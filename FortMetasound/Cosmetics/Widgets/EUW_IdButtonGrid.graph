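# Editor Utility Widget that shows a two-by-two grid of four buttons labeled as Pickaxe Ids, with one selected at a time.
# Construct stores the buttons in ButtonArray and selects the first.
# Each press calls ChangeSelection with its index, which saves ActiveId and recolors the buttons.
# It removes the chosen button from a local array copy, paints the rest UnselectedColor, then the chosen one SelectedColor.
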EditorUtilityWidgetBlueprint /FortMetasound/Cosmetics/Widgets/EUW_IdButtonGrid
graphs [
 ChangeSelection {nodes [
   n1 = ChangeSelection!() -> (then, ButtonChoice:int) {is=title K2Node_FunctionEntry}
   > Set ActiveId(ActiveId:int=@n1) -> (Output_Get:int=0)
   n2 = Get ButtonArray() -> (ButtonArray:Array of Editor Utility Button Object References)
   > n3 = Set LocalArray(LocalArray:Array of Editor Utility Button Object References=@n2) -> (Output_Get:Array of Editor Utility Button Object References)
   n4 = "Get (a copy)"(Array:Array of Editor Utility Button Object References=@n2, Dimension 1=@n1) -> (Output:Editor Utility Button Object Reference)
   > Remove Item!(execute, TargetArray:Array of Editor Utility Button Object References=@n3, Item:"Editor Utility Button Object Reference (by ref)"=@n4) -> (then, ReturnValue:bool)
   > n5 = Sequence()
   > n6 = For Each Loop(Array:Array of Editor Utility Button Object References=Get LocalArray() -> (LocalArray:Array of Editor Utility Button Object References)) -> (Array Element:Editor Utility Button Object Reference)
   > SetBackgroundColor!(
    execute
    self:Button Object Reference=@n6
    InBackgroundColor:Linear Color Structure=Get UnselectedColor() -> (UnselectedColor:Linear Color Structure)
   ) -> (then)
   SetBackgroundColor(execute=@n5.then_1, self=@n4, InBackgroundColor=Get SelectedColor() -> (SelectedColor:Linear Color Structure))
  ]}
 EventGraph {nodes [
   "On Pressed (IdButtonOne)"() {is=title K2Node_ComponentBoundEvent}
   > ChangeSelection!(execute, self, ButtonChoice:int=0) -> (then) {is=title K2Node_CallFunction}
   "On Pressed (IdButtonTwo)"() {is=title K2Node_ComponentBoundEvent}
   > ChangeSelection(ButtonChoice=1)
   "On Pressed (IdButtonThree)"() {is=title K2Node_ComponentBoundEvent}
   > ChangeSelection(ButtonChoice=2)
   "On Pressed (IdButtonFour)"() {is=title K2Node_ComponentBoundEvent}
   > ChangeSelection(ButtonChoice=3)
   Event Construct()
   > Set ButtonArray(ButtonArray:Array of Editor Utility Button Object References=Make Array!(
     "[0]":Editor Utility Button Object Reference=Get IdButtonOne() -> (IdButtonOne:Editor Utility Button Object Reference)
     "[1]":Editor Utility Button Object Reference=Get IdButtonTwo() -> (IdButtonTwo:Editor Utility Button Object Reference)
     "[2]":Editor Utility Button Object Reference=Get IdButtonThree() -> (IdButtonThree:Editor Utility Button Object Reference)
     "[3]":Editor Utility Button Object Reference=Get IdButtonFour() -> (IdButtonFour:Editor Utility Button Object Reference)
    ) -> (Array:Array of Editor Utility Button Object References)) -> (Output_Get:Array of Editor Utility Button Object References)
   > ChangeSelection(ButtonChoice=0)
  ]}
]
variables [
 ActiveId {type=integer}
 ButtonArray {type="array<EditorUtilityButton>", default []}
 SelectedColor {type=LinearColor, default=<rgba 0.2395 0.2102 0.8073 1>}
 UnselectedColor {type=LinearColor, default=<rgba 0.1425 0.0612 0.3177 1>}
]
widgets {
 info {rootWidgetClass=&CanvasPanel, widgetCount=16}
 widgets [
  CanvasPanel_1 {widgetClassPath=&CanvasPanel, widget=CanvasPanel_1}
  HorizontalBox_0 {
   widgetClassPath=&HorizontalBox
   parent=CanvasPanel_1
   widget=HorizontalBox_0
   properties {bOverride_Cursor=true}
   slot {
    class=CanvasPanelSlot
    properties {layoutData {alignment=<0 0>, anchors {maximum=<0 0>, minimum=<0 0>}, offsets=<ltrb 0 0 229.1972 242.1576>}}
   }
  }
  VerticalBox_126 {
   widgetClassPath=&VerticalBox
   parent=HorizontalBox_0
   widget=VerticalBox_126
   slot {class=HorizontalBoxSlot, properties {size {sizeRule=Fill, value=1}}}
  }
  VerticalBox_33 {
   widgetClassPath=&VerticalBox
   parent=VerticalBox_126
   widget=VerticalBox_33
   slot {class=VerticalBoxSlot, properties {size {sizeRule=Fill, value=1}}}
  }
  IdButtonOne {
   widgetClassPath=&EditorUtilityButton
   parent=VerticalBox_33
   widget=IdButtonOne
   properties {widgetStyle {
     clickedSlateSound {resourceObject=null}
     disabled {
      bIsDynamicallyLoaded=false
      drawAs=RoundedBox
      imageType=NoImage
      mirroring=NoMirror
      resourceName=None
      resourceObject=null
      tiling=NoTile
      imageSize=<0 0>
      margin=<ltrb 0 0 0 0>
      outlineSettings {
       bUseBrushTransparency=false
       roundingType=HalfHeightRadius
       width=2
       color {colorUseRule=UseColor_Specified, specifiedColor=<rgba 0 0 0 0>}
       cornerRadii=<1000 1000 1000 1000>
      }
      tintColor {colorUseRule=UseColor_Specified, specifiedColor=<rgba 0.0395 0.0395 0.0395 0>}
      uVRegion {bIsValid=false, max=<0 0>, min=<0 0>}
     }
     disabledForeground {colorUseRule=UseColor_ColorTable, specifiedColor=<rgba 0 0 0 0>}
     hovered {
      bIsDynamicallyLoaded=false
      drawAs=RoundedBox
      imageType=NoImage
      mirroring=NoMirror
      resourceName=None
      resourceObject=null
      tiling=NoTile
      imageSize=<32 32>
      margin=<ltrb 0 0 0 0>
      outlineSettings {
       bUseBrushTransparency=false
       roundingType=HalfHeightRadius
       width=2
       color {colorUseRule=UseColor_Specified, specifiedColor=<rgba 0 0 0 1>}
       cornerRadii=<1000 1000 1000 1000>
      }
      tintColor {colorUseRule=UseColor_Foreground, specifiedColor=<rgba 0.0395 0.0395 0.0395 0>}
      uVRegion {bIsValid=false, max=<0 0>, min=<0 0>}
     }
     hoveredForeground {colorUseRule=UseColor_ColorTable, specifiedColor=<rgba 0 0 0 0>}
     hoveredSlateSound {resourceObject=null}
     normal {
      bIsDynamicallyLoaded=false
      drawAs=RoundedBox
      imageType=NoImage
      mirroring=NoMirror
      resourceName=None
      resourceObject=null
      tiling=NoTile
      imageSize=<32 32>
      margin=<ltrb 0 0 0 0>
      outlineSettings {
       bUseBrushTransparency=false
       roundingType=HalfHeightRadius
       width=2
       color {colorUseRule=UseColor_Specified, specifiedColor=<rgba 0 0 0 1>}
       cornerRadii=<1000 1000 1000 1000>
      }
      tintColor {colorUseRule=UseColor_Foreground, specifiedColor=<rgba 0.0395 0.0395 0.0395 0>}
      uVRegion {bIsValid=false, max=<0 0>, min=<0 0>}
     }
     normalForeground {colorUseRule=UseColor_ColorTable, specifiedColor=<rgba 0 0 0 0>}
     normalPadding=<ltrb 12 1.5 12 1.5>
     pressed {
      bIsDynamicallyLoaded=false
      drawAs=RoundedBox
      imageType=NoImage
      mirroring=NoMirror
      resourceName=None
      resourceObject=null
      tiling=NoTile
      imageSize=<32 32>
      margin=<ltrb 0 0 0 0>
      outlineSettings {
       bUseBrushTransparency=false
       roundingType=HalfHeightRadius
       width=2
       color {colorUseRule=UseColor_Specified, specifiedColor=<rgba 0.2213 0.1518 1 1>}
       cornerRadii=<1000 1000 1000 1000>
      }
      tintColor {colorUseRule=UseColor_Foreground, specifiedColor=<rgba 0.0395 0.0395 0.0395 0>}
      uVRegion {bIsValid=false, max=<0 0>, min=<0 0>}
     }
     pressedForeground {colorUseRule=UseColor_ColorTable, specifiedColor=<rgba 0 0 0 0>}
     pressedPadding=<ltrb 12 2.5 12 0.5>
     pressedSlateSound {resourceObject=null}
    }}
   slot {class=VerticalBoxSlot, properties {padding=<ltrb 10 10 10 10>, size {sizeRule=Fill, value=1}}}
  }
  IdValue {
   widgetClassPath=&TextBlock
   bIsVariable=true
   parent=VerticalBox_33
   widget=IdValue
   properties {
    text=Pickaxe Id One
    font {
     bForceMonospaced=false
     bMaterialIsStencil=false
     fontMaterial=null
     fontObject=&Roboto
     letterSpacing=0
     monospacedWidth=1
     size=7.5
     skewAmount=0
     typefaceFontName=Bold
     outlineSettings {
      bApplyOutlineToDropShadows=false
      bMiteredCorners=false
      bSeparateFillAlpha=false
      outlineMaterial=null
      outlineSize=0
      outlineColor=<rgba 0 0 0 1>
     }
    }
   }
   slot {class=VerticalBoxSlot, properties {horizontalAlignment=HAlign_Center, verticalAlignment=VAlign_Center}}
  }
  VerticalBox_1 {
   widgetClassPath=&VerticalBox
   parent=VerticalBox_126
   widget=VerticalBox_1
   slot {class=VerticalBoxSlot, properties {size {sizeRule=Fill, value=1}}}
  }
  IdButtonThree {
   widgetClassPath=&EditorUtilityButton
   parent=VerticalBox_1
   widget=IdButtonThree
   properties {widgetStyle {
     clickedSlateSound {resourceObject=null}
     disabled {
      bIsDynamicallyLoaded=false
      drawAs=RoundedBox
      imageType=NoImage
      mirroring=NoMirror
      resourceName=None
      resourceObject=null
      tiling=NoTile
      imageSize=<0 0>
      margin=<ltrb 0 0 0 0>
      outlineSettings {
       bUseBrushTransparency=false
       roundingType=HalfHeightRadius
       width=2
       color {colorUseRule=UseColor_Specified, specifiedColor=<rgba 0 0 0 0>}
       cornerRadii=<1000 1000 1000 1000>
      }
      tintColor {colorUseRule=UseColor_Specified, specifiedColor=<rgba 0.0395 0.0395 0.0395 0>}
      uVRegion {bIsValid=false, max=<0 0>, min=<0 0>}
     }
     disabledForeground {colorUseRule=UseColor_ColorTable, specifiedColor=<rgba 0 0 0 0>}
     hovered {
      bIsDynamicallyLoaded=false
      drawAs=RoundedBox
      imageType=NoImage
      mirroring=NoMirror
      resourceName=None
      resourceObject=null
      tiling=NoTile
      imageSize=<32 32>
      margin=<ltrb 0 0 0 0>
      outlineSettings {
       bUseBrushTransparency=false
       roundingType=HalfHeightRadius
       width=2
       color {colorUseRule=UseColor_Specified, specifiedColor=<rgba 0 0 0 1>}
       cornerRadii=<1000 1000 1000 1000>
      }
      tintColor {colorUseRule=UseColor_Foreground, specifiedColor=<rgba 0.0395 0.0395 0.0395 0>}
      uVRegion {bIsValid=false, max=<0 0>, min=<0 0>}
     }
     hoveredForeground {colorUseRule=UseColor_ColorTable, specifiedColor=<rgba 0 0 0 0>}
     hoveredSlateSound {resourceObject=null}
     normal {
      bIsDynamicallyLoaded=false
      drawAs=RoundedBox
      imageType=NoImage
      mirroring=NoMirror
      resourceName=None
      resourceObject=null
      tiling=NoTile
      imageSize=<32 32>
      margin=<ltrb 0 0 0 0>
      outlineSettings {
       bUseBrushTransparency=false
       roundingType=HalfHeightRadius
       width=2
       color {colorUseRule=UseColor_Specified, specifiedColor=<rgba 0 0 0 1>}
       cornerRadii=<1000 1000 1000 1000>
      }
      tintColor {colorUseRule=UseColor_Foreground, specifiedColor=<rgba 0.0395 0.0395 0.0395 0>}
      uVRegion {bIsValid=false, max=<0 0>, min=<0 0>}
     }
     normalForeground {colorUseRule=UseColor_ColorTable, specifiedColor=<rgba 0 0 0 0>}
     normalPadding=<ltrb 12 1.5 12 1.5>
     pressed {
      bIsDynamicallyLoaded=false
      drawAs=RoundedBox
      imageType=NoImage
      mirroring=NoMirror
      resourceName=None
      resourceObject=null
      tiling=NoTile
      imageSize=<32 32>
      margin=<ltrb 0 0 0 0>
      outlineSettings {
       bUseBrushTransparency=false
       roundingType=HalfHeightRadius
       width=2
       color {colorUseRule=UseColor_Specified, specifiedColor=<rgba 0.2213 0.1518 1 1>}
       cornerRadii=<1000 1000 1000 1000>
      }
      tintColor {colorUseRule=UseColor_Foreground, specifiedColor=<rgba 0.0395 0.0395 0.0395 0>}
      uVRegion {bIsValid=false, max=<0 0>, min=<0 0>}
     }
     pressedForeground {colorUseRule=UseColor_ColorTable, specifiedColor=<rgba 0 0 0 0>}
     pressedPadding=<ltrb 12 2.5 12 0.5>
     pressedSlateSound {resourceObject=null}
    }}
   slot {class=VerticalBoxSlot, properties {padding=<ltrb 10 10 10 10>, size {sizeRule=Fill, value=1}}}
  }
  IdValue_2 {
   widgetClassPath=&TextBlock
   bIsVariable=true
   parent=VerticalBox_1
   widget=IdValue_2
   properties {
    text=Pickaxe Id Three
    font {
     bForceMonospaced=false
     bMaterialIsStencil=false
     fontMaterial=null
     fontObject=&Roboto
     letterSpacing=0
     monospacedWidth=1
     size=7.5
     skewAmount=0
     typefaceFontName=Bold
     outlineSettings {
      bApplyOutlineToDropShadows=false
      bMiteredCorners=false
      bSeparateFillAlpha=false
      outlineMaterial=null
      outlineSize=0
      outlineColor=<rgba 0 0 0 1>
     }
    }
   }
   slot {class=VerticalBoxSlot, properties {horizontalAlignment=HAlign_Center, verticalAlignment=VAlign_Center}}
  }
  VerticalBox_248 {
   widgetClassPath=&VerticalBox
   parent=HorizontalBox_0
   widget=VerticalBox_248
   slot {class=HorizontalBoxSlot, properties {size {sizeRule=Fill, value=1}}}
  }
  VerticalBox_2 {
   widgetClassPath=&VerticalBox
   parent=VerticalBox_248
   widget=VerticalBox_2
   slot {class=VerticalBoxSlot, properties {size {sizeRule=Fill, value=1}}}
  }
  IdButtonTwo {
   widgetClassPath=&EditorUtilityButton
   parent=VerticalBox_2
   widget=IdButtonTwo
   properties {widgetStyle {
     clickedSlateSound {resourceObject=null}
     disabled {
      bIsDynamicallyLoaded=false
      drawAs=RoundedBox
      imageType=NoImage
      mirroring=NoMirror
      resourceName=None
      resourceObject=null
      tiling=NoTile
      imageSize=<0 0>
      margin=<ltrb 0 0 0 0>
      outlineSettings {
       bUseBrushTransparency=false
       roundingType=HalfHeightRadius
       width=2
       color {colorUseRule=UseColor_Specified, specifiedColor=<rgba 0 0 0 0>}
       cornerRadii=<1000 1000 1000 1000>
      }
      tintColor {colorUseRule=UseColor_Specified, specifiedColor=<rgba 0.0395 0.0395 0.0395 0>}
      uVRegion {bIsValid=false, max=<0 0>, min=<0 0>}
     }
     disabledForeground {colorUseRule=UseColor_ColorTable, specifiedColor=<rgba 0 0 0 0>}
     hovered {
      bIsDynamicallyLoaded=false
      drawAs=RoundedBox
      imageType=NoImage
      mirroring=NoMirror
      resourceName=None
      resourceObject=null
      tiling=NoTile
      imageSize=<32 32>
      margin=<ltrb 0 0 0 0>
      outlineSettings {
       bUseBrushTransparency=false
       roundingType=HalfHeightRadius
       width=2
       color {colorUseRule=UseColor_Specified, specifiedColor=<rgba 0 0 0 1>}
       cornerRadii=<1000 1000 1000 1000>
      }
      tintColor {colorUseRule=UseColor_Foreground, specifiedColor=<rgba 0.0395 0.0395 0.0395 0>}
      uVRegion {bIsValid=false, max=<0 0>, min=<0 0>}
     }
     hoveredForeground {colorUseRule=UseColor_ColorTable, specifiedColor=<rgba 0 0 0 0>}
     hoveredSlateSound {resourceObject=null}
     normal {
      bIsDynamicallyLoaded=false
      drawAs=RoundedBox
      imageType=NoImage
      mirroring=NoMirror
      resourceName=None
      resourceObject=null
      tiling=NoTile
      imageSize=<32 32>
      margin=<ltrb 0 0 0 0>
      outlineSettings {
       bUseBrushTransparency=false
       roundingType=HalfHeightRadius
       width=2
       color {colorUseRule=UseColor_Specified, specifiedColor=<rgba 0 0 0 1>}
       cornerRadii=<1000 1000 1000 1000>
      }
      tintColor {colorUseRule=UseColor_Foreground, specifiedColor=<rgba 0.0395 0.0395 0.0395 0>}
      uVRegion {bIsValid=false, max=<0 0>, min=<0 0>}
     }
     normalForeground {colorUseRule=UseColor_ColorTable, specifiedColor=<rgba 0 0 0 0>}
     normalPadding=<ltrb 12 1.5 12 1.5>
     pressed {
      bIsDynamicallyLoaded=false
      drawAs=RoundedBox
      imageType=NoImage
      mirroring=NoMirror
      resourceName=None
      resourceObject=null
      tiling=NoTile
      imageSize=<32 32>
      margin=<ltrb 0 0 0 0>
      outlineSettings {
       bUseBrushTransparency=false
       roundingType=HalfHeightRadius
       width=2
       color {colorUseRule=UseColor_Specified, specifiedColor=<rgba 0.2213 0.1518 1 1>}
       cornerRadii=<1000 1000 1000 1000>
      }
      tintColor {colorUseRule=UseColor_Foreground, specifiedColor=<rgba 0.0395 0.0395 0.0395 0>}
      uVRegion {bIsValid=false, max=<0 0>, min=<0 0>}
     }
     pressedForeground {colorUseRule=UseColor_ColorTable, specifiedColor=<rgba 0 0 0 0>}
     pressedPadding=<ltrb 12 2.5 12 0.5>
     pressedSlateSound {resourceObject=null}
    }}
   slot {class=VerticalBoxSlot, properties {padding=<ltrb 10 10 10 10>, size {sizeRule=Fill, value=1}}}
  }
  IdValue_3 {
   widgetClassPath=&TextBlock
   bIsVariable=true
   parent=VerticalBox_2
   widget=IdValue_3
   properties {
    text=Pickaxe Id Two
    font {
     bForceMonospaced=false
     bMaterialIsStencil=false
     fontMaterial=null
     fontObject=&Roboto
     letterSpacing=0
     monospacedWidth=1
     size=7.5
     skewAmount=0
     typefaceFontName=Bold
     outlineSettings {
      bApplyOutlineToDropShadows=false
      bMiteredCorners=false
      bSeparateFillAlpha=false
      outlineMaterial=null
      outlineSize=0
      outlineColor=<rgba 0 0 0 1>
     }
    }
   }
   slot {class=VerticalBoxSlot, properties {horizontalAlignment=HAlign_Center, verticalAlignment=VAlign_Center}}
  }
  VerticalBox {
   widgetClassPath=&VerticalBox
   parent=VerticalBox_248
   widget=VerticalBox
   slot {class=VerticalBoxSlot, properties {size {sizeRule=Fill, value=1}}}
  }
  IdButtonFour {
   widgetClassPath=&EditorUtilityButton
   parent=VerticalBox
   widget=IdButtonFour
   properties {widgetStyle {
     clickedSlateSound {resourceObject=null}
     disabled {
      bIsDynamicallyLoaded=false
      drawAs=RoundedBox
      imageType=NoImage
      mirroring=NoMirror
      resourceName=None
      resourceObject=null
      tiling=NoTile
      imageSize=<0 0>
      margin=<ltrb 0 0 0 0>
      outlineSettings {
       bUseBrushTransparency=false
       roundingType=HalfHeightRadius
       width=2
       color {colorUseRule=UseColor_Specified, specifiedColor=<rgba 0 0 0 0>}
       cornerRadii=<1000 1000 1000 1000>
      }
      tintColor {colorUseRule=UseColor_Specified, specifiedColor=<rgba 0.0395 0.0395 0.0395 0>}
      uVRegion {bIsValid=false, max=<0 0>, min=<0 0>}
     }
     disabledForeground {colorUseRule=UseColor_ColorTable, specifiedColor=<rgba 0 0 0 0>}
     hovered {
      bIsDynamicallyLoaded=false
      drawAs=RoundedBox
      imageType=NoImage
      mirroring=NoMirror
      resourceName=None
      resourceObject=null
      tiling=NoTile
      imageSize=<32 32>
      margin=<ltrb 0 0 0 0>
      outlineSettings {
       bUseBrushTransparency=false
       roundingType=HalfHeightRadius
       width=2
       color {colorUseRule=UseColor_Specified, specifiedColor=<rgba 0 0 0 1>}
       cornerRadii=<1000 1000 1000 1000>
      }
      tintColor {colorUseRule=UseColor_Foreground, specifiedColor=<rgba 0.0395 0.0395 0.0395 0>}
      uVRegion {bIsValid=false, max=<0 0>, min=<0 0>}
     }
     hoveredForeground {colorUseRule=UseColor_ColorTable, specifiedColor=<rgba 0 0 0 0>}
     hoveredSlateSound {resourceObject=null}
     normal {
      bIsDynamicallyLoaded=false
      drawAs=RoundedBox
      imageType=NoImage
      mirroring=NoMirror
      resourceName=None
      resourceObject=null
      tiling=NoTile
      imageSize=<32 32>
      margin=<ltrb 0 0 0 0>
      outlineSettings {
       bUseBrushTransparency=false
       roundingType=HalfHeightRadius
       width=2
       color {colorUseRule=UseColor_Specified, specifiedColor=<rgba 0 0 0 1>}
       cornerRadii=<1000 1000 1000 1000>
      }
      tintColor {colorUseRule=UseColor_Foreground, specifiedColor=<rgba 0.0395 0.0395 0.0395 0>}
      uVRegion {bIsValid=false, max=<0 0>, min=<0 0>}
     }
     normalForeground {colorUseRule=UseColor_ColorTable, specifiedColor=<rgba 0 0 0 0>}
     normalPadding=<ltrb 12 1.5 12 1.5>
     pressed {
      bIsDynamicallyLoaded=false
      drawAs=RoundedBox
      imageType=NoImage
      mirroring=NoMirror
      resourceName=None
      resourceObject=null
      tiling=NoTile
      imageSize=<32 32>
      margin=<ltrb 0 0 0 0>
      outlineSettings {
       bUseBrushTransparency=false
       roundingType=HalfHeightRadius
       width=2
       color {colorUseRule=UseColor_Specified, specifiedColor=<rgba 0.2213 0.1518 1 1>}
       cornerRadii=<1000 1000 1000 1000>
      }
      tintColor {colorUseRule=UseColor_Foreground, specifiedColor=<rgba 0.0395 0.0395 0.0395 0>}
      uVRegion {bIsValid=false, max=<0 0>, min=<0 0>}
     }
     pressedForeground {colorUseRule=UseColor_ColorTable, specifiedColor=<rgba 0 0 0 0>}
     pressedPadding=<ltrb 12 2.5 12 0.5>
     pressedSlateSound {resourceObject=null}
    }}
   slot {class=VerticalBoxSlot, properties {padding=<ltrb 10 10 10 10>, size {sizeRule=Fill, value=1}}}
  }
  IdValue_1 {
   widgetClassPath=&TextBlock
   bIsVariable=true
   parent=VerticalBox
   widget=IdValue_1
   properties {
    text=Pickaxe Id Four
    font {
     bForceMonospaced=false
     bMaterialIsStencil=false
     fontMaterial=null
     fontObject=&Roboto
     letterSpacing=0
     monospacedWidth=1
     size=7.5
     skewAmount=0
     typefaceFontName=Bold
     outlineSettings {
      bApplyOutlineToDropShadows=false
      bMiteredCorners=false
      bSeparateFillAlpha=false
      outlineMaterial=null
      outlineSize=0
      outlineColor=<rgba 0 0 0 1>
     }
    }
   }
   slot {class=VerticalBoxSlot, properties {horizontalAlignment=HAlign_Center, verticalAlignment=VAlign_Center}}
  }
 ]
}
