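# Editor utility widget that serves as a MetaSound preset editor panel for a fixed set of cosmetic glider, pickaxe and reactive sounds.
# Knobs grouped as Pitch Control, Volume Control and Random Delay map to named MetaSound inputs for master multipliers, modulation ranges and start-delay range.
# It stores the supplied builder, then on Construct initializes the editor viewmodel with it and binds every knob to that viewmodel in a loop.
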EditorUtilityWidgetBlueprint /FortMetasound/Cosmetics/Widgets/EUW_MetasoundPreset_Cosmetics_Delay
graphs [
 EventGraph {nodes [
   InitializeLiteralWidgets() {is=title K2Node_CustomEvent}
   > n1 = For Each Loop(Array:Array of Meta Sound Literal Widget Interface Interfaces=Make Array!(
     "[0]":Meta Sound Literal Widget Interface Interface=Get VolumeBase_Knob() -> (VolumeBase_Knob:Meta Sound Literal Float Knob Default Object Reference)
     "[1]":Meta Sound Literal Widget Interface Interface=Get RandomVolumeMin_Knob() -> (RandomVolumeMin_Knob:Meta Sound Literal Float Knob Default Object Reference)
     "[2]":Meta Sound Literal Widget Interface Interface=Get RandomVolumeMax_Knob() -> (RandomVolumeMax_Knob:Meta Sound Literal Float Knob Default Object Reference)
     "[3]":Meta Sound Literal Widget Interface Interface=Get PitchBase_Knob() -> (PitchBase_Knob:Meta Sound Literal Float Knob Default Object Reference)
     "[4]":Meta Sound Literal Widget Interface Interface=Get RandomPitchMin_Knob() -> (RandomPitchMin_Knob:Meta Sound Literal Float Knob Default Object Reference)
     "[5]":Meta Sound Literal Widget Interface Interface=Get RandomPitchMax_Knob() -> (RandomPitchMax_Knob:Meta Sound Literal Float Knob Default Object Reference)
     "[6]":Meta Sound Literal Widget Interface Interface=Get RandomDelayMin_Knob() -> (RandomDelayMin_Knob:Meta Sound Literal Float Knob Default Object Reference)
     "[7]":Meta Sound Literal Widget Interface Interface=Get RandomDelayMax_Knob() -> (RandomDelayMax_Knob:Meta Sound Literal Float Knob Default Object Reference)
    ) -> (Array:Array of Meta Sound Literal Widget Interface Interfaces)) -> (Array Element:Meta Sound Literal Widget Interface Interface)
   > Set Literal Widget Input Viewmodels!(
    execute
    Literal Widget:Meta Sound Literal Widget Interface Interface=@n1
    MetaSound Viewmodel:Meta Sound Viewmodel Object Reference=Get MetaSoundEditorViewModel() -> (MetaSoundEditorViewModel:Meta Sound Editor Viewmodel Object Reference)
    __WorldContext:Object Reference
   ) -> (then) {is=title K2Node_CallFunction}
   Event Construct()
   > Initialize Builder!(
    execute
    self:Meta Sound Viewmodel Object Reference=Get MetaSoundEditorViewModel()
    InBuilder:Meta Sound Builder Base Object Reference=Get Builder() -> (Builder:Meta Sound Builder Base Object Reference)
   ) -> (then)
   > InitializeLiteralWidgets() {is=title K2Node_CallFunction}
   n2 = Event On MetaSoundPreset Widget Constructed!() -> (OutputDelegate, then, Builder:Meta Sound Builder Base Object Reference) {is=title K2Node_Event}
   > Set Builder(Builder:Meta Sound Builder Base Object Reference=@n2.Builder) -> (Output_Get:Meta Sound Builder Base Object Reference)
  ]}
 GetEditorName {nodes [
   GetEditorName() {is=title K2Node_FunctionEntry}
   > Return Node!(execute, ReturnValue:Text)
  ]}
 GetIconBrushName {nodes [
   GetIconBrushName() {is=title K2Node_FunctionEntry}
   > Return Node(ReturnValue:Name=None)
  ]}
 GetSupportedMetaSounds {nodes [
   GetSupportedMetaSounds() {is=title K2Node_FunctionEntry}
   > Return Node!(
    execute
    bSupportAllPresets:bool=false
    ExcludedMetaSounds:Array of Meta Sound Document Interface Interfaces
    IncludedMetaSounds:Array of Meta Sound Document Interface Interfaces=Get Included Meta Sounds() -> (Included Meta Sounds:Array of Meta Sound Document Interface Interfaces)
   )
  ]}
 OnInitializingTab {nodes [
   On Initializing Tab() {is=title K2Node_FunctionEntry}
   > Return Node!(execute, TabInfo:Meta Sound Template Widget Tab Info Structure)
  ]}
]
variables [
 Builder {type=MetaSoundBuilderBase, default=null}
 Included Meta Sounds {
  type="array<object>"
  default [
   MetaSoundSource'/FortMetasound/Cosmetics/Glider/MSS_Glider_Close'
   MetaSoundSource'/FortMetasound/Cosmetics/Glider/MSS_Glider_Open'
   MetaSoundSource'/FortMetasound/Cosmetics/Pickaxe/MSS_Pickaxe_Impact'
   MetaSoundSource'/FortMetasound/Cosmetics/Reactive/MSS_Reactive'
   MetaSoundSource'/FortMetasound/Cosmetics/Pickaxe/MSS_Pickaxe_Ready'
   MetaSoundSource'/FortMetasound/Cosmetics/Pickaxe/MSS_Pickaxe_Swing'
  ]
 }
]
widgets {
 info {rootWidgetClass=&StackBox, widgetCount=30}
 widgets [
  StackBox_0 {widgetClassPath=&StackBox, widget=StackBox_0}
  Border_1 {
   widgetClassPath=&Border
   parent=StackBox_0
   widget=Border_1
   properties {horizontalAlignment=HAlign_Left, verticalAlignment=VAlign_Top, brushColor=<rgba 0.0103 0.0103 0.0103 1>, padding=<ltrb 0 0 0 0>}
   slot {class=StackBoxSlot, properties {horizontalAlignment=HAlign_Left, verticalAlignment=VAlign_Top}}
  }
  VerticalBox_0 {
   widgetClassPath=&VerticalBox
   parent=Border_1
   widget=VerticalBox_0
   slot {class=BorderSlot, properties {horizontalAlignment=HAlign_Left, verticalAlignment=VAlign_Top, padding=<ltrb 0 0 0 0>}}
  }
  CommonTextBlock_359 {
   widgetClassPath=&CommonTextBlock
   parent=VerticalBox_0
   widget=CommonTextBlock_359
   properties {
    text=Parent Display Name
    colorAndOpacity {colorUseRule=UseColor_Specified, specifiedColor=<rgba 0.5271 0.5271 0.5271 1>}
    font {
     bForceMonospaced=false
     bMaterialIsStencil=false
     fontMaterial=null
     fontObject=&Roboto
     letterSpacing=0
     monospacedWidth=1
     size=18
     skewAmount=0
     typefaceFontName=Bold
     outlineSettings {
      bApplyOutlineToDropShadows=false
      bMiteredCorners=false
      bSeparateFillAlpha=false
      outlineMaterial=null
      outlineSize=0
      outlineColor=<rgba 0 0 0 1>
     }
    }
    shadowOffset=<0 0>
   }
   slot {class=VerticalBoxSlot, properties {horizontalAlignment=HAlign_Left, padding=<ltrb 16 16 16 8>}}
  }
  LiteralWidgets_Box {
   widgetClassPath=&WrapBox
   bIsVariable=true
   parent=VerticalBox_0
   widget=LiteralWidgets_Box
   properties {bOverride_Cursor=true, innerSlotPadding=<16 0>}
   slot {class=VerticalBoxSlot, properties {padding=<ltrb 16 8 16 16>}}
  }
  HorizontalBox_2 {
   widgetClassPath=&HorizontalBox
   parent=LiteralWidgets_Box
   widget=HorizontalBox_2
   slot {class=WrapBoxSlot, properties {verticalAlignment=VAlign_Top}}
  }
  VerticalBox_3 {widgetClassPath=&VerticalBox, parent=HorizontalBox_2, widget=VerticalBox_3, slot {class=HorizontalBoxSlot}}
  Border_0 {
   widgetClassPath=&Border
   parent=VerticalBox_3
   widget=Border_0
   properties {
    horizontalAlignment=HAlign_Left
    verticalAlignment=VAlign_Top
    background {
     bIsDynamicallyLoaded=false
     drawAs=RoundedBox
     imageType=NoImage
     mirroring=NoMirror
     resourceName=None
     resourceObject=null
     tiling=NoTile
     imageSize=<32 32>
     margin=<ltrb 0 0 0 0>
     outlineSettings {
      bUseBrushTransparency=false
      roundingType=FixedRadius
      width=0
      color {colorUseRule=UseColor_Specified, specifiedColor=<rgba 0.0075 0.0075 0.0075 1>}
      cornerRadii=<12 12 12 12>
     }
     tintColor {colorUseRule=UseColor_Specified, specifiedColor=<rgba 1 1 1 1>}
     uVRegion {bIsValid=false, max=<0 0>, min=<0 0>}
    }
    brushColor=<rgba 0.0212 0.0212 0.0212 1>
    padding=<ltrb 16 16 16 16>
   }
   slot {class=VerticalBoxSlot}
  }
  VerticalBox_1 {
   widgetClassPath=&VerticalBox
   parent=Border_0
   widget=VerticalBox_1
   slot {class=BorderSlot, properties {horizontalAlignment=HAlign_Left, verticalAlignment=VAlign_Top, padding=<ltrb 16 16 16 16>}}
  }
  CommonTextBlock_272 {
   widgetClassPath=&CommonTextBlock
   parent=VerticalBox_1
   widget=CommonTextBlock_272
   properties {
    text=Pitch Control
    colorAndOpacity {colorUseRule=UseColor_Specified, specifiedColor=<rgba 0.5271 0.5271 0.5271 1>}
    font {
     bForceMonospaced=false
     bMaterialIsStencil=false
     fontMaterial=null
     fontObject=&Roboto
     letterSpacing=0
     monospacedWidth=1
     size=9
     skewAmount=0
     typefaceFontName=Regular
     outlineSettings {
      bApplyOutlineToDropShadows=false
      bMiteredCorners=false
      bSeparateFillAlpha=false
      outlineMaterial=null
      outlineSize=0
      outlineColor=<rgba 0 0 0 1>
     }
    }
    shadowOffset=<0 0>
   }
   slot {class=VerticalBoxSlot, properties {horizontalAlignment=HAlign_Center, padding=<ltrb 8 8 8 0>}}
  }
  WrapBox {
   widgetClassPath=&WrapBox
   bIsVariable=true
   parent=VerticalBox_1
   widget=WrapBox
   properties {innerSlotPadding=<16 0>}
   slot {class=VerticalBoxSlot, properties {horizontalAlignment=HAlign_Left, verticalAlignment=VAlign_Top, padding=<ltrb 8 0 8 0>}}
  }
  PitchBase_Knob {
   widgetClassPath=/TechAudioToolsContent/Widgets/MetaSoundLiterals/Float/EUW_MetaSoundFloat_Knob_Default
   bIsVariable=true
   parent=WrapBox
   widget=PitchBase_Knob
   properties {
    default Value=1
    input Name=TemplatePitchMultiplier
    label Text=Master
    metaSoundLiteralViewModel_Float=&self:WidgetTree.PitchBase_Knob.MetaSoundLiteralViewModel_Float_0
    knob Style {
     material=null
     desiredSize=<96 96>
     knobAccentColor=<rgba 0.06 0.06 0.06 1>
     knobBarColor=<rgba 0.067 0.067 0.067 1>
     knobBarFillMaxColor=<rgba 0 0.0153 0.2656 1>
     knobBarFillMidColor=<rgba 0 0.0422 0.7344 1>
     knobBarFillMinColor=<rgba 0.1094 0.1606 1 1>
     knobBarFillTintColor=<rgba 0.96 0.96 0.96 1>
     knobBarShadowColor=<rgba 0.067 0.067 0.067 1>
     knobEdgeFillColor=<rgba 0.0156 0.0156 0.0156 1>
     knobIndicatorDotColor=<rgba 0.0102 0.0039 0.1042 1>
     knobMainColor=<rgba 0.1406 0.1406 0.1406 1>
     knobShadowColor=<rgba 0.06 0.06 0.06 1>
     knobSmoothBevelColor=<rgba 0.0417 0.0417 0.0417 1>
     textBoxStyle {
      backgroundColor {colorUseRule=UseColor_ColorTable, specifiedColor=<rgba 0 0 0 0>}
      backgroundImage {
       bIsDynamicallyLoaded=false
       drawAs=RoundedBox
       imageType=NoImage
       mirroring=NoMirror
       resourceName=None
       resourceObject=null
       tiling=NoTile
       imageSize=<56 28>
       margin=<ltrb 0 0 0 0>
       outlineSettings {
        bUseBrushTransparency=false
        roundingType=FixedRadius
        width=0
        color {colorUseRule=UseColor_Specified, specifiedColor=<rgba 0 0 0 0>}
        cornerRadii=<4 4 4 4>
       }
       tintColor {colorUseRule=UseColor_ColorTable, specifiedColor=<rgba 0 0 0 0>}
       uVRegion {bIsValid=false, max=<0 0>, min=<0 0>}
      }
     }
    }
   }
   slot {class=WrapBoxSlot}
  }
  RandomPitchMin_Knob {
   widgetClassPath=&EUW_MetaSoundFloat_Knob_Default
   bIsVariable=true
   parent=WrapBox
   widget=RandomPitchMin_Knob
   properties {
    default Value=0.95
    input Name=PitchModulationRange.Min
    label Text=Min
    metaSoundLiteralViewModel_Float=&self:WidgetTree.RandomPitchMin_Knob.MetaSoundLiteralViewModel_Float_0
    knob Style {
     material=null
     desiredSize=<96 96>
     knobAccentColor=<rgba 0.06 0.06 0.06 1>
     knobBarColor=<rgba 0.067 0.067 0.067 1>
     knobBarFillMaxColor=<rgba 0 0.0153 0.2656 1>
     knobBarFillMidColor=<rgba 0 0.0422 0.7344 1>
     knobBarFillMinColor=<rgba 0.1094 0.1606 1 1>
     knobBarFillTintColor=<rgba 0.96 0.96 0.96 1>
     knobBarShadowColor=<rgba 0.067 0.067 0.067 1>
     knobEdgeFillColor=<rgba 0.0156 0.0156 0.0156 1>
     knobIndicatorDotColor=<rgba 0.0102 0.0039 0.1042 1>
     knobMainColor=<rgba 0.1406 0.1406 0.1406 1>
     knobShadowColor=<rgba 0.06 0.06 0.06 1>
     knobSmoothBevelColor=<rgba 0.0417 0.0417 0.0417 1>
     textBoxStyle {
      backgroundColor {colorUseRule=UseColor_ColorTable, specifiedColor=<rgba 0 0 0 0>}
      backgroundImage {
       bIsDynamicallyLoaded=false
       drawAs=RoundedBox
       imageType=NoImage
       mirroring=NoMirror
       resourceName=None
       resourceObject=null
       tiling=NoTile
       imageSize=<56 28>
       margin=<ltrb 0 0 0 0>
       outlineSettings {
        bUseBrushTransparency=false
        roundingType=FixedRadius
        width=0
        color {colorUseRule=UseColor_Specified, specifiedColor=<rgba 0 0 0 0>}
        cornerRadii=<4 4 4 4>
       }
       tintColor {colorUseRule=UseColor_ColorTable, specifiedColor=<rgba 0 0 0 0>}
       uVRegion {bIsValid=false, max=<0 0>, min=<0 0>}
      }
     }
    }
   }
   slot {class=WrapBoxSlot}
  }
  RandomPitchMax_Knob {
   widgetClassPath=&EUW_MetaSoundFloat_Knob_Default
   bIsVariable=true
   parent=WrapBox
   widget=RandomPitchMax_Knob
   properties {
    default Value=1.05
    input Name=PitchModulationRange.Max
    label Text=Max
    metaSoundLiteralViewModel_Float=&self:WidgetTree.RandomPitchMax_Knob.MetaSoundLiteralViewModel_Float_0
    knob Style {
     material=null
     desiredSize=<96 96>
     knobAccentColor=<rgba 0.06 0.06 0.06 1>
     knobBarColor=<rgba 0.067 0.067 0.067 1>
     knobBarFillMaxColor=<rgba 0 0.0153 0.2656 1>
     knobBarFillMidColor=<rgba 0 0.0422 0.7344 1>
     knobBarFillMinColor=<rgba 0.1094 0.1606 1 1>
     knobBarFillTintColor=<rgba 0.96 0.96 0.96 1>
     knobBarShadowColor=<rgba 0.067 0.067 0.067 1>
     knobEdgeFillColor=<rgba 0.0156 0.0156 0.0156 1>
     knobIndicatorDotColor=<rgba 0.0102 0.0039 0.1042 1>
     knobMainColor=<rgba 0.1406 0.1406 0.1406 1>
     knobShadowColor=<rgba 0.06 0.06 0.06 1>
     knobSmoothBevelColor=<rgba 0.0417 0.0417 0.0417 1>
     textBoxStyle {
      backgroundColor {colorUseRule=UseColor_ColorTable, specifiedColor=<rgba 0 0 0 0>}
      backgroundImage {
       bIsDynamicallyLoaded=false
       drawAs=RoundedBox
       imageType=NoImage
       mirroring=NoMirror
       resourceName=None
       resourceObject=null
       tiling=NoTile
       imageSize=<56 28>
       margin=<ltrb 0 0 0 0>
       outlineSettings {
        bUseBrushTransparency=false
        roundingType=FixedRadius
        width=0
        color {colorUseRule=UseColor_Specified, specifiedColor=<rgba 0 0 0 0>}
        cornerRadii=<4 4 4 4>
       }
       tintColor {colorUseRule=UseColor_ColorTable, specifiedColor=<rgba 0 0 0 0>}
       uVRegion {bIsValid=false, max=<0 0>, min=<0 0>}
      }
     }
    }
   }
   slot {class=WrapBoxSlot}
  }
  CommonTextBlock {
   widgetClassPath=&CommonTextBlock
   parent=VerticalBox_1
   widget=CommonTextBlock
   properties {
    text=Linear
    colorAndOpacity {colorUseRule=UseColor_Specified, specifiedColor=<rgba 0.5271 0.5271 0.5271 1>}
    font {
     bForceMonospaced=false
     bMaterialIsStencil=false
     fontMaterial=null
     fontObject=&Roboto
     letterSpacing=0
     monospacedWidth=1
     size=9
     skewAmount=0
     typefaceFontName=Regular
     outlineSettings {
      bApplyOutlineToDropShadows=false
      bMiteredCorners=false
      bSeparateFillAlpha=false
      outlineMaterial=null
      outlineSize=0
      outlineColor=<rgba 0 0 0 1>
     }
    }
    shadowOffset=<0 0>
   }
   slot {class=VerticalBoxSlot, properties {horizontalAlignment=HAlign_Center, padding=<ltrb 8 0 8 8>}}
  }
  Border_2 {
   widgetClassPath=&Border
   parent=VerticalBox_3
   widget=Border_2
   properties {
    horizontalAlignment=HAlign_Left
    verticalAlignment=VAlign_Top
    background {
     bIsDynamicallyLoaded=false
     drawAs=RoundedBox
     imageType=NoImage
     mirroring=NoMirror
     resourceName=None
     resourceObject=null
     tiling=NoTile
     imageSize=<32 32>
     margin=<ltrb 0 0 0 0>
     outlineSettings {
      bUseBrushTransparency=false
      roundingType=FixedRadius
      width=0
      color {colorUseRule=UseColor_Specified, specifiedColor=<rgba 0.0075 0.0075 0.0075 1>}
      cornerRadii=<12 12 12 12>
     }
     tintColor {colorUseRule=UseColor_Specified, specifiedColor=<rgba 1 1 1 1>}
     uVRegion {bIsValid=false, max=<0 0>, min=<0 0>}
    }
    brushColor=<rgba 0.0212 0.0212 0.0212 1>
    padding=<ltrb 16 16 16 16>
   }
   slot {class=VerticalBoxSlot, properties {verticalAlignment=VAlign_Top, padding=<ltrb 0 20 0 0>}}
  }
  VerticalBox_2 {
   widgetClassPath=&VerticalBox
   parent=Border_2
   widget=VerticalBox_2
   slot {class=BorderSlot, properties {horizontalAlignment=HAlign_Left, verticalAlignment=VAlign_Top, padding=<ltrb 16 16 16 16>}}
  }
  CommonTextBlock_1 {
   widgetClassPath=&CommonTextBlock
   parent=VerticalBox_2
   widget=CommonTextBlock_1
   properties {
    text=Volume Control
    colorAndOpacity {colorUseRule=UseColor_Specified, specifiedColor=<rgba 0.5271 0.5271 0.5271 1>}
    font {
     bForceMonospaced=false
     bMaterialIsStencil=false
     fontMaterial=null
     fontObject=&Roboto
     letterSpacing=0
     monospacedWidth=1
     size=9
     skewAmount=0
     typefaceFontName=Regular
     outlineSettings {
      bApplyOutlineToDropShadows=false
      bMiteredCorners=false
      bSeparateFillAlpha=false
      outlineMaterial=null
      outlineSize=0
      outlineColor=<rgba 0 0 0 1>
     }
    }
    shadowOffset=<0 0>
   }
   slot {class=VerticalBoxSlot, properties {horizontalAlignment=HAlign_Center, padding=<ltrb 8 8 8 0>}}
  }
  WrapBox_1 {
   widgetClassPath=&WrapBox
   bIsVariable=true
   parent=VerticalBox_2
   widget=WrapBox_1
   properties {innerSlotPadding=<16 0>}
   slot {class=VerticalBoxSlot, properties {horizontalAlignment=HAlign_Left, verticalAlignment=VAlign_Top, padding=<ltrb 8 0 8 0>}}
  }
  VolumeBase_Knob {
   widgetClassPath=&EUW_MetaSoundFloat_Knob_Default
   bIsVariable=true
   parent=WrapBox_1
   widget=VolumeBase_Knob
   properties {
    default Value=1
    input Name=TemplateVolumeMultiplier
    label Text=Master
    metaSoundLiteralViewModel_Float=&self:WidgetTree.VolumeBase_Knob.MetaSoundLiteralViewModel_Float_0
    knob Style {
     material=null
     desiredSize=<96 96>
     knobAccentColor=<rgba 0.06 0.06 0.06 1>
     knobBarColor=<rgba 0.067 0.067 0.067 1>
     knobBarFillMaxColor=<rgba 0.8707 1 0 1>
     knobBarFillMidColor=<rgba 0.3169 1 0 1>
     knobBarFillMinColor=<rgba 0 1 0.0698 1>
     knobBarFillTintColor=<rgba 0.96 0.96 0.96 1>
     knobBarShadowColor=<rgba 0.067 0.067 0.067 1>
     knobEdgeFillColor=<rgba 0.0156 0.0156 0.0156 1>
     knobIndicatorDotColor=<rgba 0 1 0.0698 1>
     knobMainColor=<rgba 0.1406 0.1406 0.1406 1>
     knobShadowColor=<rgba 0.06 0.06 0.06 1>
     knobSmoothBevelColor=<rgba 0.0417 0.0417 0.0417 1>
     textBoxStyle {
      backgroundColor {colorUseRule=UseColor_ColorTable, specifiedColor=<rgba 0 0 0 0>}
      backgroundImage {
       bIsDynamicallyLoaded=false
       drawAs=RoundedBox
       imageType=NoImage
       mirroring=NoMirror
       resourceName=None
       resourceObject=null
       tiling=NoTile
       imageSize=<56 28>
       margin=<ltrb 0 0 0 0>
       outlineSettings {
        bUseBrushTransparency=false
        roundingType=FixedRadius
        width=0
        color {colorUseRule=UseColor_Specified, specifiedColor=<rgba 0 0 0 0>}
        cornerRadii=<4 4 4 4>
       }
       tintColor {colorUseRule=UseColor_ColorTable, specifiedColor=<rgba 0 0 0 0>}
       uVRegion {bIsValid=false, max=<0 0>, min=<0 0>}
      }
     }
    }
   }
   slot {class=WrapBoxSlot}
  }
  RandomVolumeMin_Knob {
   widgetClassPath=&EUW_MetaSoundFloat_Knob_Default
   bIsVariable=true
   parent=WrapBox_1
   widget=RandomVolumeMin_Knob
   properties {
    default Value=0.95
    input Name=VolumeModulationRange.Min
    label Text=Min
    metaSoundLiteralViewModel_Float=&self:WidgetTree.RandomVolumeMin_Knob.MetaSoundLiteralViewModel_Float_0
    knob Style {
     material=null
     desiredSize=<96 96>
     knobAccentColor=<rgba 0.06 0.06 0.06 1>
     knobBarColor=<rgba 0.067 0.067 0.067 1>
     knobBarFillMaxColor=<rgba 0.8707 1 0 1>
     knobBarFillMidColor=<rgba 0.3169 1 0 1>
     knobBarFillMinColor=<rgba 0 1 0.0698 1>
     knobBarFillTintColor=<rgba 0.96 0.96 0.96 1>
     knobBarShadowColor=<rgba 0.067 0.067 0.067 1>
     knobEdgeFillColor=<rgba 0.0156 0.0156 0.0156 1>
     knobIndicatorDotColor=<rgba 0 1 0.0698 1>
     knobMainColor=<rgba 0.1406 0.1406 0.1406 1>
     knobShadowColor=<rgba 0.06 0.06 0.06 1>
     knobSmoothBevelColor=<rgba 0.0417 0.0417 0.0417 1>
     textBoxStyle {
      backgroundColor {colorUseRule=UseColor_ColorTable, specifiedColor=<rgba 0 0 0 0>}
      backgroundImage {
       bIsDynamicallyLoaded=false
       drawAs=RoundedBox
       imageType=NoImage
       mirroring=NoMirror
       resourceName=None
       resourceObject=null
       tiling=NoTile
       imageSize=<56 28>
       margin=<ltrb 0 0 0 0>
       outlineSettings {
        bUseBrushTransparency=false
        roundingType=FixedRadius
        width=0
        color {colorUseRule=UseColor_Specified, specifiedColor=<rgba 0 0 0 0>}
        cornerRadii=<4 4 4 4>
       }
       tintColor {colorUseRule=UseColor_ColorTable, specifiedColor=<rgba 0 0 0 0>}
       uVRegion {bIsValid=false, max=<0 0>, min=<0 0>}
      }
     }
    }
   }
   slot {class=WrapBoxSlot}
  }
  RandomVolumeMax_Knob {
   widgetClassPath=&EUW_MetaSoundFloat_Knob_Default
   bIsVariable=true
   parent=WrapBox_1
   widget=RandomVolumeMax_Knob
   properties {
    default Value=1.05
    input Name=VolumeModulationRange.Max
    label Text=Max
    metaSoundLiteralViewModel_Float=&self:WidgetTree.RandomVolumeMax_Knob.MetaSoundLiteralViewModel_Float_0
    knob Style {
     material=null
     desiredSize=<96 96>
     knobAccentColor=<rgba 0.06 0.06 0.06 1>
     knobBarColor=<rgba 0.067 0.067 0.067 1>
     knobBarFillMaxColor=<rgba 0.8707 1 0 1>
     knobBarFillMidColor=<rgba 0.3169 1 0 1>
     knobBarFillMinColor=<rgba 0 1 0.0698 1>
     knobBarFillTintColor=<rgba 0.96 0.96 0.96 1>
     knobBarShadowColor=<rgba 0.067 0.067 0.067 1>
     knobEdgeFillColor=<rgba 0.0156 0.0156 0.0156 1>
     knobIndicatorDotColor=<rgba 0 1 0.0698 1>
     knobMainColor=<rgba 0.1406 0.1406 0.1406 1>
     knobShadowColor=<rgba 0.06 0.06 0.06 1>
     knobSmoothBevelColor=<rgba 0.0417 0.0417 0.0417 1>
     textBoxStyle {
      backgroundColor {colorUseRule=UseColor_ColorTable, specifiedColor=<rgba 0 0 0 0>}
      backgroundImage {
       bIsDynamicallyLoaded=false
       drawAs=RoundedBox
       imageType=NoImage
       mirroring=NoMirror
       resourceName=None
       resourceObject=null
       tiling=NoTile
       imageSize=<56 28>
       margin=<ltrb 0 0 0 0>
       outlineSettings {
        bUseBrushTransparency=false
        roundingType=FixedRadius
        width=0
        color {colorUseRule=UseColor_Specified, specifiedColor=<rgba 0 0 0 0>}
        cornerRadii=<4 4 4 4>
       }
       tintColor {colorUseRule=UseColor_ColorTable, specifiedColor=<rgba 0 0 0 0>}
       uVRegion {bIsValid=false, max=<0 0>, min=<0 0>}
      }
     }
    }
   }
   slot {class=WrapBoxSlot}
  }
  CommonTextBlock_2 {
   widgetClassPath=&CommonTextBlock
   parent=VerticalBox_2
   widget=CommonTextBlock_2
   properties {
    text=Linear
    colorAndOpacity {colorUseRule=UseColor_Specified, specifiedColor=<rgba 0.5271 0.5271 0.5271 1>}
    font {
     bForceMonospaced=false
     bMaterialIsStencil=false
     fontMaterial=null
     fontObject=&Roboto
     letterSpacing=0
     monospacedWidth=1
     size=9
     skewAmount=0
     typefaceFontName=Regular
     outlineSettings {
      bApplyOutlineToDropShadows=false
      bMiteredCorners=false
      bSeparateFillAlpha=false
      outlineMaterial=null
      outlineSize=0
      outlineColor=<rgba 0 0 0 1>
     }
    }
    shadowOffset=<0 0>
   }
   slot {class=VerticalBoxSlot, properties {horizontalAlignment=HAlign_Center, padding=<ltrb 8 0 8 8>}}
  }
  Border {
   widgetClassPath=&Border
   parent=HorizontalBox_2
   widget=Border
   properties {
    horizontalAlignment=HAlign_Left
    verticalAlignment=VAlign_Top
    background {
     bIsDynamicallyLoaded=false
     drawAs=RoundedBox
     imageType=NoImage
     mirroring=NoMirror
     resourceName=None
     resourceObject=null
     tiling=NoTile
     imageSize=<32 32>
     margin=<ltrb 0 0 0 0>
     outlineSettings {
      bUseBrushTransparency=false
      roundingType=FixedRadius
      width=0
      color {colorUseRule=UseColor_Specified, specifiedColor=<rgba 0.0075 0.0075 0.0075 1>}
      cornerRadii=<12 12 12 12>
     }
     tintColor {colorUseRule=UseColor_Specified, specifiedColor=<rgba 1 1 1 1>}
     uVRegion {bIsValid=false, max=<0 0>, min=<0 0>}
    }
    brushColor=<rgba 0.0212 0.0212 0.0212 1>
    padding=<ltrb 16 16 16 16>
   }
   slot {class=HorizontalBoxSlot, properties {horizontalAlignment=HAlign_Left, verticalAlignment=VAlign_Top, padding=<ltrb 20 0 0 0>}}
  }
  VerticalBox {
   widgetClassPath=&VerticalBox
   parent=Border
   widget=VerticalBox
   slot {class=BorderSlot, properties {horizontalAlignment=HAlign_Left, verticalAlignment=VAlign_Top, padding=<ltrb 16 16 16 16>}}
  }
  CommonTextBlock_3 {
   widgetClassPath=&CommonTextBlock
   parent=VerticalBox
   widget=CommonTextBlock_3
   properties {
    text=Random Delay
    colorAndOpacity {colorUseRule=UseColor_Specified, specifiedColor=<rgba 0.5271 0.5271 0.5271 1>}
    font {
     bForceMonospaced=false
     bMaterialIsStencil=false
     fontMaterial=null
     fontObject=&Roboto
     letterSpacing=0
     monospacedWidth=1
     size=9
     skewAmount=0
     typefaceFontName=Regular
     outlineSettings {
      bApplyOutlineToDropShadows=false
      bMiteredCorners=false
      bSeparateFillAlpha=false
      outlineMaterial=null
      outlineSize=0
      outlineColor=<rgba 0 0 0 1>
     }
    }
    shadowOffset=<0 0>
   }
   slot {class=VerticalBoxSlot, properties {horizontalAlignment=HAlign_Center, padding=<ltrb 8 8 8 0>}}
  }
  WrapBox_2 {
   widgetClassPath=&WrapBox
   bIsVariable=true
   parent=VerticalBox
   widget=WrapBox_2
   properties {innerSlotPadding=<16 0>}
   slot {class=VerticalBoxSlot, properties {horizontalAlignment=HAlign_Left, verticalAlignment=VAlign_Top, padding=<ltrb 8 0 8 0>}}
  }
  RandomDelayMin_Knob {
   widgetClassPath=&EUW_MetaSoundFloat_Knob_Default
   bIsVariable=true
   parent=WrapBox_2
   widget=RandomDelayMin_Knob
   properties {
    input Name=StartDelay.Min
    label Text=Min
    metaSoundLiteralViewModel_Float=&self:WidgetTree.RandomDelayMin_Knob.MetaSoundLiteralViewModel_Float_0
    knob Style {
     material=null
     desiredSize=<96 96>
     knobAccentColor=<rgba 0.06 0.06 0.06 1>
     knobBarColor=<rgba 0.067 0.067 0.067 1>
     knobBarFillMaxColor=<rgba 0.2969 0 0.0132 1>
     knobBarFillMidColor=<rgba 0.7031 0 0.0314 1>
     knobBarFillMinColor=<rgba 1 0.1979 0.2337 1>
     knobBarFillTintColor=<rgba 0.96 0.96 0.96 1>
     knobBarShadowColor=<rgba 0.067 0.067 0.067 1>
     knobEdgeFillColor=<rgba 0.0156 0.0156 0.0156 1>
     knobIndicatorDotColor=<rgba 1 0.0354 0.0785 1>
     knobMainColor=<rgba 0.1406 0.1406 0.1406 1>
     knobShadowColor=<rgba 0.06 0.06 0.06 1>
     knobSmoothBevelColor=<rgba 0.0417 0.0417 0.0417 1>
     textBoxStyle {
      backgroundColor {colorUseRule=UseColor_ColorTable, specifiedColor=<rgba 0 0 0 0>}
      backgroundImage {
       bIsDynamicallyLoaded=false
       drawAs=RoundedBox
       imageType=NoImage
       mirroring=NoMirror
       resourceName=None
       resourceObject=null
       tiling=NoTile
       imageSize=<56 28>
       margin=<ltrb 0 0 0 0>
       outlineSettings {
        bUseBrushTransparency=false
        roundingType=FixedRadius
        width=0
        color {colorUseRule=UseColor_Specified, specifiedColor=<rgba 0 0 0 0>}
        cornerRadii=<4 4 4 4>
       }
       tintColor {colorUseRule=UseColor_ColorTable, specifiedColor=<rgba 0 0 0 0>}
       uVRegion {bIsValid=false, max=<0 0>, min=<0 0>}
      }
     }
    }
   }
   slot {class=WrapBoxSlot}
  }
  RandomDelayMax_Knob {
   widgetClassPath=&EUW_MetaSoundFloat_Knob_Default
   bIsVariable=true
   parent=WrapBox_2
   widget=RandomDelayMax_Knob
   properties {
    input Name=StartDelay.Max
    label Text=Max
    metaSoundLiteralViewModel_Float=&self:WidgetTree.RandomDelayMax_Knob.MetaSoundLiteralViewModel_Float_0
    knob Style {
     material=null
     desiredSize=<96 96>
     knobAccentColor=<rgba 0.06 0.06 0.06 1>
     knobBarColor=<rgba 0.067 0.067 0.067 1>
     knobBarFillMaxColor=<rgba 0.2969 0 0.0132 1>
     knobBarFillMidColor=<rgba 0.7031 0 0.0314 1>
     knobBarFillMinColor=<rgba 1 0.1979 0.2337 1>
     knobBarFillTintColor=<rgba 0.96 0.96 0.96 1>
     knobBarShadowColor=<rgba 0.067 0.067 0.067 1>
     knobEdgeFillColor=<rgba 0.0156 0.0156 0.0156 1>
     knobIndicatorDotColor=<rgba 1 0.0354 0.0785 1>
     knobMainColor=<rgba 0.1406 0.1406 0.1406 1>
     knobShadowColor=<rgba 0.06 0.06 0.06 1>
     knobSmoothBevelColor=<rgba 0.0417 0.0417 0.0417 1>
     textBoxStyle {
      backgroundColor {colorUseRule=UseColor_ColorTable, specifiedColor=<rgba 0 0 0 0>}
      backgroundImage {
       bIsDynamicallyLoaded=false
       drawAs=RoundedBox
       imageType=NoImage
       mirroring=NoMirror
       resourceName=None
       resourceObject=null
       tiling=NoTile
       imageSize=<56 28>
       margin=<ltrb 0 0 0 0>
       outlineSettings {
        bUseBrushTransparency=false
        roundingType=FixedRadius
        width=0
        color {colorUseRule=UseColor_Specified, specifiedColor=<rgba 0 0 0 0>}
        cornerRadii=<4 4 4 4>
       }
       tintColor {colorUseRule=UseColor_ColorTable, specifiedColor=<rgba 0 0 0 0>}
       uVRegion {bIsValid=false, max=<0 0>, min=<0 0>}
      }
     }
    }
   }
   slot {class=WrapBoxSlot}
  }
  CommonTextBlock_4 {
   widgetClassPath=&CommonTextBlock
   parent=VerticalBox
   widget=CommonTextBlock_4
   properties {
    text=Linear
    colorAndOpacity {colorUseRule=UseColor_Specified, specifiedColor=<rgba 0.5271 0.5271 0.5271 1>}
    font {
     bForceMonospaced=false
     bMaterialIsStencil=false
     fontMaterial=null
     fontObject=&Roboto
     letterSpacing=0
     monospacedWidth=1
     size=9
     skewAmount=0
     typefaceFontName=Regular
     outlineSettings {
      bApplyOutlineToDropShadows=false
      bMiteredCorners=false
      bSeparateFillAlpha=false
      outlineMaterial=null
      outlineSize=0
      outlineColor=<rgba 0 0 0 1>
     }
    }
    shadowOffset=<0 0>
   }
   slot {class=VerticalBoxSlot, properties {horizontalAlignment=HAlign_Center, padding=<ltrb 8 0 8 8>}}
  }
 ]
}
