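# Editor utility widget that provides a circular drag pad for setting four MetaSound float inputs (Forward, Sideways, Backwards, Thrust), bound through configurable input names.
# While the Dragger is held, Tick moves it with the mouse, clamped to the circle using cached widget geometry.
# Its offset from center becomes normalized axis values, with unsigned Sideways and distance-based Thrust, written to the viewmodels and broadcast via OnValuesSet.
# A timer-detected double-click recenters the Dragger and zeroes all values.
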EditorUtilityWidgetBlueprint /FortMetasound/Cosmetics/Widgets/EUW_DragControl_Literal
graphs [
 EventGraph {nodes [
   "On Released (Dragger)"() {is=title K2Node_ComponentBoundEvent}
   > Set Selected(Selected:bool=false) -> (Output_Get:bool)
   "On Clicked (Dragger)"() {is=title K2Node_ComponentBoundEvent}
   > n1 = Get!(execute:Exec) -> (then:Exec, else:Exec, bDoubleClicked:bool)
   > SetPosition!(
    execute
    self:Canvas Panel Slot Object Reference=SlotAsCanvasSlot!(Widget:Widget Object Reference=Get Dragger() -> (Dragger:Editor Utility Button Object Reference)) -> (ReturnValue:Canvas Panel Slot Object Reference)
    InPosition:Vector 2D Structure=vector2d - vector2d(A:Vector 2D Structure=Get Center() -> (Center:Vector 2D Structure), B:Vector 2D Structure=Get TopLeft() -> (TopLeft:Vector 2D Structure)) -> (ReturnValue:Vector 2D Structure) {is=title K2Node_PromotableOperator}
   ) -> (then)
   > Set Selected(Selected=false)
   > Set bDoubleClicked(bDoubleClicked:bool=false) -> (Output_Get:bool)
   > Set Reset(Reset:bool=true) -> (Output_Get:bool)
   > OnReset()
   > Set Reset(Reset=false)
   n2 = ClearDoubleClick() {is=title K2Node_CustomEvent}
   Set Timer by Event(execute=@n1.else, Delegate=@n2, Time=0.2, bLooping=false)
   > Set bDoubleClicked(bDoubleClicked=true)
   "On Pressed (Dragger)"() {is=title K2Node_ComponentBoundEvent}
   > Set Selected(Selected=true)
   TrackLocation() {is=title K2Node_CustomEvent}
   > n3 = GetMousePositionOnViewport!(execute, WorldContextObject:Object Reference) -> (then, ReturnValue:Vector 2D Structure)
   n4 = Get Center()
   n5 = Distance2D!(V1:Vector 2D Structure=@n3, V2:Vector 2D Structure=@n4) -> (ReturnValue:double)
   n6 = Get Radius() -> (Radius:double)
   n7 = Get TopLeft()
   > SetPosition(
    self=SlotAsCanvasSlot(Widget=Get Dragger())
    InPosition=Select(
     Option 0:Vector 2D Structure=Outside Circle Range!(A:Vector 2D Structure=@n3, B:Vector 2D Structure=@n4) -> (ReturnValue:Vector 2D Structure) {is=title K2Node_MacroInstance}
     Option 1:Vector 2D Structure=vector2d - vector2d(A=@n3, B=@n7)
     Index="In Range (Float)"!(Value:double=@n5, Min:double=0, Max:double=@n6, InclusiveMin:bool=true, InclusiveMax:bool=true) -> (ReturnValue:bool)
    ) -> (ReturnValue:Vector 2D Structure)
   )
   GetWidgetLocation() {is=title K2Node_CustomEvent}
   n8 = GetCachedGeometry!(self:Widget Object Reference=Get EditorUtilityButton_15() -> (EditorUtilityButton_15:Editor Utility Button Object Reference)) -> (ReturnValue:"Geometry Structure (by ref)")
   > Set TopLeft(TopLeft:Vector 2D Structure="Local to Absolute (Coordinates)"!(
     Geometry:"Geometry Structure (by ref)"=@n8
     LocalCoordinate:Vector 2D Structure=GetLocalTopLeft!(Geometry:"Geometry Structure (by ref)"=@n8) -> (ReturnValue:Vector 2D Structure)
    ) -> (ReturnValue:Vector 2D Structure)) -> (Output_Get:Vector 2D Structure)
   n9 = GetLocalSize!(Geometry:"Geometry Structure (by ref)"=@n8) -> (ReturnValue:Vector 2D Structure)
   > Set BottomRight(BottomRight:Vector 2D Structure="Local to Absolute (Coordinates)"(Geometry=@n8, LocalCoordinate=@n9)) -> (Output_Get:Vector 2D Structure)
   > Set Radius(Radius:double=float / float(A=BreakVector2D(InVec=@n9), B:int=2)) -> (Output_Get:double)
   n10 = Get TopLeft()
   n11 = Get BottomRight() -> (BottomRight:Vector 2D Structure)
   > Set Center(Center:Vector 2D Structure=vector2d / vector2d(
     A:Vector 2D Structure=vector2d + vector2d(A:Vector 2D Structure=@n10, B:Vector 2D Structure=@n11) -> (ReturnValue:Vector 2D Structure) {is=title K2Node_PromotableOperator}
     B=2
    ) -> (ReturnValue:Vector 2D Structure) {is=title K2Node_PromotableOperator}) -> (Output_Get:Vector 2D Structure)
   > SetBackgroundColor!(
    execute
    self:Button Object Reference=Get EditorUtilityButton_15()
    InBackgroundColor:Linear Color Structure=Get Inactive() -> (Inactive:Linear Color Structure)
   ) -> (then)
   n12 = OnReset() {is=title K2Node_CustomEvent}
   n13 = Sequence(execute=[@n37.then, @n12.then])
   n14 = Get Dragger()
   n15 = vector2d + vector2d(
    A=GetAbsolutePosition!(Geometry:"Geometry Structure (by ref)"=GetCachedGeometry(self=@n14)) -> (ReturnValue:Vector 2D Structure)
    B=(X=25,Y=25)
   )
   n16 = BreakVector2D(InVec=@n15)
   n17 = Get Center()
   n18 = BreakVector2D(InVec=@n17)
   n19 = float - float(A=@n18.Y, B=@n16.Y)
   n20 = Get Radius()
   > Set SourceValue!(
    execute
    SourceValue:float=SelectFloat(A=0, B=MapRangeClamped(Value=@n19, InRangeA=0, InRangeB=@n20), bPickA=Get Reset() -> (Reset:bool))
    self:Meta Sound Literal Float Viewmodel Object Reference=Get MetaSoundLiteralViewModel_Float_Forward() -> (MetaSoundLiteralViewModel_Float_Forward:Meta Sound Literal Float Viewmodel Object Reference)
   ) -> (then, Output_Get:float)
   n21 = Sequence(execute=@n13.then_1)
   n22 = "Absolute (Float)"(A=float - float(A=@n18, B=@n16))
   n23 = Get Radius()
   > Set SourceValue(
    SourceValue=SelectFloat(A=0, B=MapRangeClamped(Value=@n22, InRangeA=0, InRangeB=@n23), bPickA=Get Reset())
    self=Get MetaSoundLiteralViewModel_Float_Sideways() -> (MetaSoundLiteralViewModel_Float_Sideways:Meta Sound Literal Float Viewmodel Object Reference)
   )
   n24 = Sequence(execute=@n21.then_1)
   n25 = float * float(A=Get Radius(), B:int=-1) {is=title K2Node_PromotableOperator}
   > Set SourceValue(
    SourceValue=SelectFloat(A=0, B=MapRangeClamped(Value=@n19, InRangeA=0, InRangeB=@n25), bPickA=Get Reset())
    self=Get MetaSoundLiteralViewModel_Float_Backwards() -> (MetaSoundLiteralViewModel_Float_Backwards:Meta Sound Literal Float Viewmodel Object Reference)
   )
   n26 = Sequence(execute=@n24.then_1)
   n27 = Get Radius()
   n28 = Distance2D(V1=@n15, V2=@n17)
   > Set SourceValue(
    SourceValue=SelectFloat(A=0, B=MapRangeClamped(Value=@n28, InRangeA=0, InRangeB=@n27), bPickA=Get Reset())
    self=Get MetaSoundLiteralViewModel_Float_Thrust() -> (MetaSoundLiteralViewModel_Float_Thrust:Meta Sound Literal Float Viewmodel Object Reference)
   )
   Sequence(execute=@n26.then_1)
   n29 = Get MetaSoundLiteralViewModel_Float_Thrust()
   n30 = Get MetaSoundLiteralViewModel_Float_Forward()
   n31 = Get MetaSoundLiteralViewModel_Float_Sideways()
   n32 = Get MetaSoundLiteralViewModel_Float_Backwards()
   > Set DirectonalValues(DirectonalValues:S Drag Control Values Structure=Make S Drag Control Values!(
     Forward_8_#1:double=Get SourceValue!(self:Meta Sound Literal Float Viewmodel Object Reference=@n30) -> (SourceValue:float)
     Sideways_9_#2:double=Get SourceValue(self=@n31)
     Backward_10_#3:double=Get SourceValue(self=@n32)
     Thrust_11_#4:double=Get SourceValue(self=@n29)
    ) -> (S_DragControl_Values:S Drag Control Values Structure) {is=title K2Node_MakeStruct}) -> (Output_Get:S Drag Control Values Structure)
   > Call On Values Set!(execute:Exec, self:EUW Drag Control Literal Object Reference) -> (then:Exec) {is=title K2Node_CallDelegate}
   n33 = Event Set Input Viewmodels!() -> (OutputDelegate, then, InputViewModels:Map of Names to Meta Sound Input Viewmodel Object References) {is=title K2Node_Event}
   > n34 = Sequence()
   > SetMetaSoundInputViewModel_Forward!(
    execute
    self
    Viewmodel:Meta Sound Input Viewmodel Object Reference=Find(
     TargetMap:Map of Names to Meta Sound Input Viewmodel Object References=@n33.InputViewModels
     Key:"Name (by ref)"=Get ForwardName() -> (ForwardName:Name=None)
    ) -> (Value:Meta Sound Input Viewmodel Object Reference)
   ) -> (then)
   n35 = Sequence(execute=@n34.then_1)
   > SetMetaSoundInputViewModel_Sideways!(
    execute
    self
    Viewmodel:Meta Sound Input Viewmodel Object Reference=Find(TargetMap=@n33.InputViewModels, Key=Get SidewaysName() -> (SidewaysName:Name=None))
   ) -> (then)
   n36 = Sequence(execute=@n35.then_1)
   > SetMetaSoundInputViewModel_Backwards!(
    execute
    self
    Viewmodel:Meta Sound Input Viewmodel Object Reference=Find(TargetMap=@n33.InputViewModels, Key=Get BackwardsName() -> (BackwardsName:Name=None))
   ) -> (then)
   Sequence(execute=@n36.then_1)
   > SetMetaSoundInputViewModel_Thrust!(
    execute
    self
    Viewmodel:Meta Sound Input Viewmodel Object Reference=Find(TargetMap=@n33.InputViewModels, Key=Get ThrustName() -> (ThrustName:Name=None))
   ) -> (then)
   Event Construct()
   > Delay!(execute, WorldContextObject:Object Reference, Duration=0.1) -> (then)
   > GetWidgetLocation() {is=title K2Node_CallFunction}
   Event Tick()
   > Get!(execute) -> (then, else, Selected:bool)
   > TrackLocation() {is=title K2Node_CallFunction}
   > n37 = GetWidgetLocation()
   Set bDoubleClicked(execute=@n2.then, bDoubleClicked=false)
  ]}
 GetEditorName {nodes [
   GetEditorName() {is=title K2Node_FunctionEntry}
   > Return Node!(execute, ReturnValue:Text)
  ]}
 GetIconBrushName {nodes [
   GetIconBrushName() {is=title K2Node_FunctionEntry}
   > Return Node(ReturnValue:Name=None)
  ]}
 GetInputViewModelNames {nodes [
   Get Input Viewmodel Names() {is=title K2Node_FunctionEntry}
   > Return Node(ReturnValue:Array of Names=Make Array!(
     "[0]":Name=Get ForwardName() -> (ForwardName=None)
     "[1]":Name=Get SidewaysName() -> (SidewaysName=None)
     "[2]":Name=Get BackwardsName() -> (BackwardsName=None)
     "[3]":Name=Get ThrustName() -> (ThrustName=None)
    ) -> (Array:Array of Names))
  ]}
 OnInitializingTab {nodes [
   OnInitializingTab() {is=title K2Node_FunctionEntry}
   > Return Node!(execute, TabInfo:Meta Sound Template Widget Tab Info Structure)
  ]}
 OnMouseDoubleClick {nodes [
   OnMouseDoubleClick!() -> (then, MyGeometry:Geometry Structure, MouseEvent:"Pointer Event Structure (by ref)") {is=title K2Node_FunctionEntry}
   > Return Node!(execute, ReturnValue:Event Reply Structure=Make Event Reply!() -> (EventReply:Event Reply Structure) {is=title K2Node_MakeStruct})
  ]}
 OnMouseMove {nodes [
   n1 = OnMouseMove!() -> (then, MyGeometry:Geometry Structure, MouseEvent:"Pointer Event Structure (by ref)") {is=title K2Node_FunctionEntry}
   > n2 = Parent: OnMouseMove!(execute, MyGeometry:Geometry Structure=@n1, MouseEvent:"Pointer Event Structure (by ref)"=@n1.MouseEvent) -> (then, ReturnValue:Event Reply Structure) {is=title K2Node_CallParentFunction}
   > Return Node(ReturnValue=@n2)
  ]}
 OnValuesSet {nodes [OnValuesSet() {is=title K2Node_FunctionEntry}]}
 OutsideCircleRange {nodes [
   n1 = Inputs!() -> (A:Vector 2D Structure, B:Vector 2D Structure)
   n2 = vector2d - vector2d(A=@n1, B=@n1.B)
   n3 = Vector2D Length!(A:Vector 2D Structure=@n2) -> (ReturnValue:double)
   n4 = vector2d * vector2d(
    A:Vector 2D Structure=Normalize 2D!(A:Vector 2D Structure=vector2d / vector2d(A=@n2, B=@n3)) -> (ReturnValue:Vector 2D Structure)
    B=Get Radius()
   ) -> (ReturnValue:Vector 2D Structure) {is=title K2Node_PromotableOperator}
   n5 = Get Center()
   Outputs!(ReturnValue:Vector 2D Structure=vector2d - vector2d(A=vector2d + vector2d(A=@n5, B=@n4), B=Get TopLeft()))
  ]}
]
variables [
 Active {type=LinearColor, default=<rgba 0.1407 0.163 0.4792 1>}
 BackwardsName {type=string, default=None}
 BackwardValue {type=number}
 bDoubleClicked {type=boolean}
 BottomRight {type=Vector2D, default=<0 0>}
 Center {type=Vector2D, default=<0 0>}
 DirectonalValues {type=S_DragControl_Values, default {backward=0, forward=0, sideways=0, thrust=0}}
 DoubleClickTimer {type=TimerHandle}
 EditorViewModel {type=MetaSoundInputViewModel, default=null}
 ForwardName {type=string, default=None}
 ForwardValue {type=number}
 In Position {type=Vector2D, default=<0 0>}
 Inactive {type=LinearColor, default=<rgba 0.0566 0.0655 0.1927 1>}
 OnValuesSet {default=()}
 Radius {type=number}
 Reset {type=boolean}
 Selected {type=boolean}
 SidewaysName {type=string, default=None}
 SidewaysValue {type=number}
 ThrustName {type=string, default=None}
 ThrustValue {type=number}
 TimerHandle {type=TimerHandle}
 TopLeft {type=Vector2D, default=<0 0>}
]
widgets {
 info {rootWidgetClass=&CanvasPanel, widgetCount=13}
 widgets [
  CanvasPanel_297 {widgetClassPath=&CanvasPanel, widget=CanvasPanel_297, properties {bOverride_Cursor=true}}
  CanvasPanel_1 {
   widgetClassPath=&CanvasPanel
   parent=CanvasPanel_297
   widget=CanvasPanel_1
   slot {class=CanvasPanelSlot, properties {layoutData {alignment=<0 0>, anchors {maximum=<0 0>, minimum=<0 0>}, offsets=<ltrb 0 0 500 500>}}}
  }
  MainCircle {
   widgetClassPath=&Border
   bIsVariable=true
   parent=CanvasPanel_1
   widget=MainCircle
   properties {background {
     bIsDynamicallyLoaded=false
     drawAs=RoundedBox
     imageType=NoImage
     mirroring=NoMirror
     resourceName=None
     resourceObject=null
     tiling=NoTile
     imageSize=<32 32>
     margin=<ltrb 0 0 0 0>
     outlineSettings {
      bUseBrushTransparency=false
      roundingType=HalfHeightRadius
      width=5
      color {colorUseRule=UseColor_Specified, specifiedColor=<rgba 0 0 0 1>}
      cornerRadii=<50 50 50 50>
     }
     tintColor {colorUseRule=UseColor_Specified, specifiedColor=<rgba 0.026 0.026 0.026 0>}
     uVRegion {bIsValid=false, max=<0 0>, min=<0 0>}
    }}
   slot {
    class=CanvasPanelSlot
    properties {zOrder=2, layoutData {alignment=<0 0>, anchors {maximum=<0 0>, minimum=<0 0>}, offsets=<ltrb 0 0 500.0184 500.0677>}}
   }
  }
  Background {
   widgetClassPath=&Border
   bIsVariable=true
   parent=CanvasPanel_1
   widget=Background
   properties {background {
     bIsDynamicallyLoaded=false
     drawAs=RoundedBox
     imageType=NoImage
     mirroring=NoMirror
     resourceName=None
     resourceObject=null
     tiling=NoTile
     imageSize=<32 32>
     margin=<ltrb 0 0 0 0>
     outlineSettings {
      bUseBrushTransparency=false
      roundingType=HalfHeightRadius
      width=5
      color {colorUseRule=UseColor_Specified, specifiedColor=<rgba 0 0 0 1>}
      cornerRadii=<50 50 50 50>
     }
     tintColor {colorUseRule=UseColor_Specified, specifiedColor=<rgba 0.026 0.026 0.026 1>}
     uVRegion {bIsValid=false, max=<0 0>, min=<0 0>}
    }}
   slot {
    class=CanvasPanelSlot
    properties {layoutData {alignment=<0 0>, anchors {maximum=<0 0>, minimum=<0 0>}, offsets=<ltrb 0 0 500.0184 500.0677>}}
   }
  }
  Border13 {
   widgetClassPath=&Border
   parent=CanvasPanel_1
   widget=Border13
   properties {background {
     bIsDynamicallyLoaded=false
     drawAs=RoundedBox
     imageType=NoImage
     mirroring=NoMirror
     resourceName=None
     resourceObject=null
     tiling=NoTile
     imageSize=<32 32>
     margin=<ltrb 0 0 0 0>
     outlineSettings {
      bUseBrushTransparency=false
      roundingType=HalfHeightRadius
      width=5
      color {colorUseRule=UseColor_Specified, specifiedColor=<rgba 0 0 0 1>}
      cornerRadii=<50 50 50 50>
     }
     tintColor {colorUseRule=UseColor_Specified, specifiedColor=<rgba 0.026 0.026 0.026 0>}
     uVRegion {bIsValid=false, max=<0 0>, min=<0 0>}
    }}
   slot {
    class=CanvasPanelSlot
    properties {zOrder=2, layoutData {alignment=<0.5 0.5>, anchors {maximum=<0 0>, minimum=<0 0>}, offsets=<ltrb 250 250 166 166>}}
   }
  }
  Border23 {
   widgetClassPath=&Border
   parent=CanvasPanel_1
   widget=Border23
   properties {background {
     bIsDynamicallyLoaded=false
     drawAs=RoundedBox
     imageType=NoImage
     mirroring=NoMirror
     resourceName=None
     resourceObject=null
     tiling=NoTile
     imageSize=<32 32>
     margin=<ltrb 0 0 0 0>
     outlineSettings {
      bUseBrushTransparency=false
      roundingType=HalfHeightRadius
      width=5
      color {colorUseRule=UseColor_Specified, specifiedColor=<rgba 0 0 0 1>}
      cornerRadii=<50 50 50 50>
     }
     tintColor {colorUseRule=UseColor_Specified, specifiedColor=<rgba 0.026 0.026 0.026 0>}
     uVRegion {bIsValid=false, max=<0 0>, min=<0 0>}
    }}
   slot {
    class=CanvasPanelSlot
    properties {zOrder=2, layoutData {alignment=<0.5 0.5>, anchors {maximum=<0 0>, minimum=<0 0>}, offsets=<ltrb 250 250 333 333>}}
   }
  }
  VerticalLine {
   widgetClassPath=&Border
   parent=CanvasPanel_1
   widget=VerticalLine
   properties {
    background {
     bIsDynamicallyLoaded=false
     drawAs=RoundedBox
     imageType=NoImage
     mirroring=NoMirror
     resourceName=None
     resourceObject=null
     tiling=NoTile
     imageSize=<32 32>
     margin=<ltrb 0 0 0 0>
     outlineSettings {
      bUseBrushTransparency=false
      roundingType=FixedRadius
      width=5
      color {colorUseRule=UseColor_Specified, specifiedColor=<rgba 0 0 0 1>}
      cornerRadii=<0 0 0 0>
     }
     tintColor {colorUseRule=UseColor_Specified, specifiedColor=<rgba 0.026 0.026 0.026 0>}
     uVRegion {bIsValid=false, max=<0 0>, min=<0 0>}
    }
    brushColor=<rgba 0 0 0 1>
   }
   slot {
    class=CanvasPanelSlot
    properties {zOrder=1, layoutData {alignment=<0.5 0.5>, anchors {maximum=<0 0>, minimum=<0 0>}, offsets=<ltrb 250 250 5 500>}}
   }
  }
  DiagonalOne {
   widgetClassPath=&Border
   parent=CanvasPanel_1
   widget=DiagonalOne
   properties {
    background {
     bIsDynamicallyLoaded=false
     drawAs=RoundedBox
     imageType=NoImage
     mirroring=NoMirror
     resourceName=None
     resourceObject=null
     tiling=NoTile
     imageSize=<32 32>
     margin=<ltrb 0 0 0 0>
     outlineSettings {
      bUseBrushTransparency=false
      roundingType=FixedRadius
      width=5
      color {colorUseRule=UseColor_Specified, specifiedColor=<rgba 0 0 0 1>}
      cornerRadii=<0 0 0 0>
     }
     tintColor {colorUseRule=UseColor_Specified, specifiedColor=<rgba 0.026 0.026 0.026 0>}
     uVRegion {bIsValid=false, max=<0 0>, min=<0 0>}
    }
    renderTransform {angle=45, scale=<1 1>, shear=<0 0>, translation=<0 0>}
   }
   slot {
    class=CanvasPanelSlot
    properties {zOrder=1, layoutData {alignment=<0.5 0.5>, anchors {maximum=<0 0>, minimum=<0 0>}, offsets=<ltrb 250 250 5 500>}}
   }
  }
  DiagonalTwo {
   widgetClassPath=&Border
   parent=CanvasPanel_1
   widget=DiagonalTwo
   properties {
    background {
     bIsDynamicallyLoaded=false
     drawAs=RoundedBox
     imageType=NoImage
     mirroring=NoMirror
     resourceName=None
     resourceObject=null
     tiling=NoTile
     imageSize=<32 32>
     margin=<ltrb 0 0 0 0>
     outlineSettings {
      bUseBrushTransparency=false
      roundingType=FixedRadius
      width=5
      color {colorUseRule=UseColor_Specified, specifiedColor=<rgba 0 0 0 1>}
      cornerRadii=<0 0 0 0>
     }
     tintColor {colorUseRule=UseColor_Specified, specifiedColor=<rgba 0.026 0.026 0.026 0>}
     uVRegion {bIsValid=false, max=<0 0>, min=<0 0>}
    }
    renderTransform {angle=-45, scale=<1 1>, shear=<0 0>, translation=<0 0>}
   }
   slot {
    class=CanvasPanelSlot
    properties {zOrder=1, layoutData {alignment=<0.5 0.5>, anchors {maximum=<0 0>, minimum=<0 0>}, offsets=<ltrb 250 250 5 500>}}
   }
  }
  HorizontalLine {
   widgetClassPath=&Border
   parent=CanvasPanel_1
   widget=HorizontalLine
   properties {background {
     bIsDynamicallyLoaded=false
     drawAs=RoundedBox
     imageType=NoImage
     mirroring=NoMirror
     resourceName=None
     resourceObject=null
     tiling=NoTile
     imageSize=<32 32>
     margin=<ltrb 0 0 0 0>
     outlineSettings {
      bUseBrushTransparency=false
      roundingType=FixedRadius
      width=5
      color {colorUseRule=UseColor_Specified, specifiedColor=<rgba 0 0 0 1>}
      cornerRadii=<0 0 0 0>
     }
     tintColor {colorUseRule=UseColor_Specified, specifiedColor=<rgba 0.026 0.026 0.026 0>}
     uVRegion {bIsValid=false, max=<0 0>, min=<0 0>}
    }}
   slot {
    class=CanvasPanelSlot
    properties {zOrder=1, layoutData {alignment=<0.5 0.5>, anchors {maximum=<0 0>, minimum=<0 0>}, offsets=<ltrb 250 250 500 5>}}
   }
  }
  EditorUtilityButton_15 {
   widgetClassPath=&EditorUtilityButton
   parent=CanvasPanel_1
   widget=EditorUtilityButton_15
   properties {widgetStyle {
     clickedSlateSound {resourceObject=null}
     disabled {
      bIsDynamicallyLoaded=false
      drawAs=RoundedBox
      imageType=NoImage
      mirroring=NoMirror
      resourceName=None
      resourceObject=null
      tiling=NoTile
      imageSize=<0 0>
      margin=<ltrb 0 0 0 0>
      outlineSettings {
       bUseBrushTransparency=false
       roundingType=HalfHeightRadius
       width=1
       color {colorUseRule=UseColor_ColorTable, specifiedColor=<rgba 0 0 0 0>}
       cornerRadii=<1000 1000 1000 1000>
      }
      tintColor {colorUseRule=UseColor_Specified, specifiedColor=<rgba 0.0395 0.0395 0.0395 0>}
      uVRegion {bIsValid=false, max=<0 0>, min=<0 0>}
     }
     disabledForeground {colorUseRule=UseColor_ColorTable, specifiedColor=<rgba 0 0 0 0>}
     hovered {
      bIsDynamicallyLoaded=false
      drawAs=RoundedBox
      imageType=NoImage
      mirroring=NoMirror
      resourceName=None
      resourceObject=null
      tiling=NoTile
      imageSize=<32 32>
      margin=<ltrb 0 0 0 0>
      outlineSettings {
       bUseBrushTransparency=false
       roundingType=HalfHeightRadius
       width=1
       color {colorUseRule=UseColor_Specified, specifiedColor=<rgba 0 0 0 0>}
       cornerRadii=<1000 1000 1000 1000>
      }
      tintColor {colorUseRule=UseColor_Specified, specifiedColor=<rgba 0.0395 0.0395 0.0395 0>}
      uVRegion {bIsValid=false, max=<0 0>, min=<0 0>}
     }
     hoveredForeground {colorUseRule=UseColor_ColorTable, specifiedColor=<rgba 0 0 0 0>}
     hoveredSlateSound {resourceObject=null}
     normal {
      bIsDynamicallyLoaded=false
      drawAs=RoundedBox
      imageType=NoImage
      mirroring=NoMirror
      resourceName=None
      resourceObject=null
      tiling=NoTile
      imageSize=<32 32>
      margin=<ltrb 0 0 0 0>
      outlineSettings {
       bUseBrushTransparency=false
       roundingType=HalfHeightRadius
       width=1
       color {colorUseRule=UseColor_Specified, specifiedColor=<rgba 0 0 0 0>}
       cornerRadii=<1000 1000 1000 1000>
      }
      tintColor {colorUseRule=UseColor_Specified, specifiedColor=<rgba 0.0395 0.0395 0.0395 0>}
      uVRegion {bIsValid=false, max=<0 0>, min=<0 0>}
     }
     normalForeground {colorUseRule=UseColor_ColorTable, specifiedColor=<rgba 0 0 0 0>}
     normalPadding=<ltrb 12 1.5 12 1.5>
     pressed {
      bIsDynamicallyLoaded=false
      drawAs=RoundedBox
      imageType=NoImage
      mirroring=NoMirror
      resourceName=None
      resourceObject=null
      tiling=NoTile
      imageSize=<32 32>
      margin=<ltrb 0 0 0 0>
      outlineSettings {
       bUseBrushTransparency=false
       roundingType=HalfHeightRadius
       width=1
       color {colorUseRule=UseColor_Specified, specifiedColor=<rgba 0 0 0 0>}
       cornerRadii=<1000 1000 1000 1000>
      }
      tintColor {colorUseRule=UseColor_Specified, specifiedColor=<rgba 0.0395 0.0395 0.0395 0>}
      uVRegion {bIsValid=false, max=<0 0>, min=<0 0>}
     }
     pressedForeground {colorUseRule=UseColor_ColorTable, specifiedColor=<rgba 0 0 0 0>}
     pressedPadding=<ltrb 12 2.5 12 0.5>
     pressedSlateSound {resourceObject=null}
    }}
   slot {
    class=CanvasPanelSlot
    properties {zOrder=3, layoutData {alignment=<0 0>, anchors {maximum=<0 0>, minimum=<0 0>}, offsets=<ltrb 0 0 500 500>}}
   }
  }
  Dragger {
   widgetClassPath=&EditorUtilityButton
   parent=CanvasPanel_1
   widget=Dragger
   properties {
    backgroundColor=<rgba 0.0941 0 1 1>
    colorAndOpacity=<rgba 0.0941 0 1 1>
    widgetStyle {
     clickedSlateSound {resourceObject=null}
     disabled {
      bIsDynamicallyLoaded=false
      drawAs=RoundedBox
      imageType=FullColor
      mirroring=NoMirror
      resourceName=None
      resourceObject=null
      tiling=NoTile
      imageSize=<128 128>
      margin=<ltrb 0 0 0 0>
      outlineSettings {
       bUseBrushTransparency=false
       roundingType=HalfHeightRadius
       width=1
       color {colorUseRule=UseColor_Specified, specifiedColor=<rgba 0.1038 0 1 1>}
       cornerRadii=<1000 1000 1000 1000>
      }
      tintColor {colorUseRule=UseColor_Specified, specifiedColor=<rgba 0.0395 0.0395 0.0395 1>}
      uVRegion {bIsValid=false, max=<0 0>, min=<0 0>}
     }
     disabledForeground {colorUseRule=UseColor_Specified, specifiedColor=<rgba 0.0941 0 1 1>}
     hovered {
      bIsDynamicallyLoaded=false
      drawAs=RoundedBox
      imageType=FullColor
      mirroring=NoMirror
      resourceName=None
      resourceObject=null
      tiling=NoTile
      imageSize=<128 128>
      margin=<ltrb 0 0 0 0>
      outlineSettings {
       bUseBrushTransparency=false
       roundingType=HalfHeightRadius
       width=1
       color {colorUseRule=UseColor_Specified, specifiedColor=<rgba 0.2111 0.1198 1 1>}
       cornerRadii=<1000 1000 1000 1000>
      }
      tintColor {colorUseRule=UseColor_Specified, specifiedColor=<rgba 0.0395 0.0395 0.0395 1>}
      uVRegion {bIsValid=false, max=<0 0>, min=<0 0>}
     }
     hoveredForeground {colorUseRule=UseColor_Specified, specifiedColor=<rgba 0.0941 0 1 1>}
     hoveredSlateSound {resourceObject=null}
     normal {
      bIsDynamicallyLoaded=false
      drawAs=RoundedBox
      imageType=FullColor
      mirroring=NoMirror
      resourceName=None
      resourceObject=null
      tiling=NoTile
      imageSize=<128 128>
      margin=<ltrb 0 0 0 0>
      outlineSettings {
       bUseBrushTransparency=false
       roundingType=HalfHeightRadius
       width=1
       color {colorUseRule=UseColor_Specified, specifiedColor=<rgba 0.1038 0 1 1>}
       cornerRadii=<1000 1000 1000 1000>
      }
      tintColor {colorUseRule=UseColor_Specified, specifiedColor=<rgba 0.0395 0.0395 0.0395 1>}
      uVRegion {bIsValid=false, max=<0 0>, min=<0 0>}
     }
     normalForeground {colorUseRule=UseColor_Specified, specifiedColor=<rgba 0.0941 0 1 0>}
     normalPadding=<ltrb 12 1.5 12 1.5>
     pressed {
      bIsDynamicallyLoaded=false
      drawAs=RoundedBox
      imageType=FullColor
      mirroring=NoMirror
      resourceName=None
      resourceObject=null
      tiling=NoTile
      imageSize=<128 128>
      margin=<ltrb 0 0 0 0>
      outlineSettings {
       bUseBrushTransparency=false
       roundingType=HalfHeightRadius
       width=1
       color {colorUseRule=UseColor_Specified, specifiedColor=<rgba 0.3978 0.3281 1 1>}
       cornerRadii=<1000 1000 1000 1000>
      }
      tintColor {colorUseRule=UseColor_Specified, specifiedColor=<rgba 0.0395 0.0395 0.0395 1>}
      uVRegion {bIsValid=false, max=<0 0>, min=<0 0>}
     }
     pressedForeground {colorUseRule=UseColor_Specified, specifiedColor=<rgba 0.0941 0 1 1>}
     pressedPadding=<ltrb 12 2.5 12 0.5>
     pressedSlateSound {resourceObject=null}
    }
   }
   slot {
    class=CanvasPanelSlot
    properties {zOrder=10, layoutData {alignment=<0.5 0.5>, anchors {maximum=<0 0>, minimum=<0 0>}, offsets=<ltrb 250 250 50 50>}}
   }
  }
  TextBlock_33 {
   widgetClassPath=&TextBlock
   bIsVariable=true
   parent=CanvasPanel_297
   widget=TextBlock_33
   properties {visibility=Hidden}
   slot {class=CanvasPanelSlot, properties=~}
  }
 ]
}
