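# Editor utility widget that puts a drag control over four MetaSound float inputs: Forward, Sideways, Backwards and Thrust.
# The embedded EUW_DragControl_Literal fires On Values Set, and UpdateEvent writes its DirectonalValues into four float literal viewmodels and text readouts.
# Four name variables define the requested inputs; Set Input Viewmodels looks them up and assigns each as an editor viewmodel.
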EditorUtilityWidgetBlueprint /FortMetasound/Cosmetics/Widgets/EUW_DragControl_Default
graphs [
 EventGraph {nodes [
   n1 = UpdateEvent() {is=title K2Node_CustomEvent}
   > n2 = Sequence()
   n3 = Get EUW_DragControl_Literal() -> (EUW_DragControl_Literal:EUW Drag Control Literal Object Reference)
   n4 = Break S Drag Control Values!(S_DragControl_Values:"S Drag Control Values Structure (by ref)"=Get DirectonalValues!(self:EUW Drag Control Literal Object Reference=@n3) -> (DirectonalValues:S Drag Control Values Structure)) -> (
    Forward_8_#1:double=~
    Sideways_9_#2:double=~
    Backward_10_#3:double=~
    Thrust_11_#4:double=~
   ) {is=title K2Node_BreakStruct}
   > n5 = Set DisplayValue!(
    execute
    DisplayValue:float=@n4
    self:Meta Sound Literal Float Viewmodel Object Reference=Get MetaSoundLiteralViewModel_Float_Forward() -> (MetaSoundLiteralViewModel_Float_Forward:Meta Sound Literal Float Viewmodel Object Reference)
   ) -> (then, Output_Get:float)
   > Set Text(
    Text="To Text (Float)"(Value=@n5, MinimumIntegralDigits=1, MaximumIntegralDigits=324, MinimumFractionalDigits=0, MaximumFractionalDigits=3)
    self=Get ForwardsDisplay() -> (ForwardsDisplay:Text Object Reference)
   )
   n6 = Sequence(execute=@n2.then_1)
   > n7 = Set DisplayValue(
    DisplayValue=@n4.Sideways_9_#2
    self=Get MetaSoundLiteralViewModel_Float_Sideways() -> (MetaSoundLiteralViewModel_Float_Sideways:Meta Sound Literal Float Viewmodel Object Reference)
   )
   > Set Text(
    Text="To Text (Float)"(Value=@n7, MinimumIntegralDigits=1, MaximumIntegralDigits=324, MinimumFractionalDigits=0, MaximumFractionalDigits=3)
    self=Get SidewaysDisplay() -> (SidewaysDisplay:Text Object Reference)
   )
   n8 = Sequence(execute=@n6.then_1)
   > n9 = Set DisplayValue(
    DisplayValue=@n4.Backward_10_#3
    self=Get MetaSoundLiteralViewModel_Float_Backwards() -> (MetaSoundLiteralViewModel_Float_Backwards:Meta Sound Literal Float Viewmodel Object Reference)
   )
   > Set Text(
    Text="To Text (Float)"(Value=@n9, MinimumIntegralDigits=1, MaximumIntegralDigits=324, MinimumFractionalDigits=0, MaximumFractionalDigits=3)
    self=Get BackwardsDisplay() -> (BackwardsDisplay:Text Object Reference)
   )
   Sequence(execute=@n8.then_1)
   > n10 = Set DisplayValue(
    DisplayValue=@n4.Thrust_11_#4
    self=Get MetaSoundLiteralViewModel_Float_Thrust() -> (MetaSoundLiteralViewModel_Float_Thrust:Meta Sound Literal Float Viewmodel Object Reference)
   )
   > Set Text(
    Text="To Text (Float)"(Value=@n10, MinimumIntegralDigits=1, MaximumIntegralDigits=324, MinimumFractionalDigits=0, MaximumFractionalDigits=3)
    self=Get ThustDispaly() -> (ThustDispaly:Text Object Reference)
   )
   Event Construct()
   > Bind Event to On Values Set(self:EUW Drag Control Literal Object Reference=@n3, Delegate=@n1) {is=title K2Node_AddDelegate}
   n11 = Event Set Input Viewmodels!() -> (OutputDelegate, then, InputViewModels:Map of Names to Meta Sound Input Viewmodel Object References) {is=title K2Node_Event}
   > n12 = Sequence()
   > n13 = Cast To MetaSoundInputEditorViewModel!(
    execute
    Object=Find(
     TargetMap:Map of Names to Meta Sound Input Viewmodel Object References=@n11.InputViewModels
     Key:"Name (by ref)"=Get ForwardName() -> (ForwardName:Name=None)
    ) -> (Value:Meta Sound Input Viewmodel Object Reference)
   ) -> (then, CastFailed, AsMetaSound Input Editor Viewmodel:Meta Sound Input Editor Viewmodel Object Reference) {is=title K2Node_DynamicCast}
   > SetMetaSoundInputEditorViewModel_Forward!(execute, self, Viewmodel:Meta Sound Input Editor Viewmodel Object Reference=@n13) -> (then)
   n14 = Sequence(execute=@n12.then_1)
   > n15 = Cast To MetaSoundInputEditorViewModel(Object=Find(TargetMap=@n11.InputViewModels, Key=Get SidewaysName() -> (SidewaysName:Name=None)))
   > SetMetaSoundInputEditorViewModel_Sideways!(execute, self, Viewmodel:Meta Sound Input Editor Viewmodel Object Reference=@n15) -> (then)
   n16 = Sequence(execute=@n14.then_1)
   > n17 = Cast To MetaSoundInputEditorViewModel(Object=Find(TargetMap=@n11.InputViewModels, Key=Get BackwardsName() -> (BackwardsName:Name=None)))
   > SetMetaSoundInputEditorViewModel_Backwards!(execute, self, Viewmodel:Meta Sound Input Editor Viewmodel Object Reference=@n17) -> (then)
   Sequence(execute=@n16.then_1)
   > n18 = Cast To MetaSoundInputEditorViewModel(Object=Find(TargetMap=@n11.InputViewModels, Key=Get ThrustName() -> (ThrustName:Name=None)))
   > SetMetaSoundInputEditorViewModel_Thrust!(execute, self, Viewmodel:Meta Sound Input Editor Viewmodel Object Reference=@n18) -> (then)
   Event PreConstruct()
  ]}
 GetInputViewModelNames {nodes [
   Get Input Viewmodel Names() {is=title K2Node_FunctionEntry}
   > Return Node!(
    execute
    ReturnValue:Array of Names=Make Array!(
     "[0]":Name=Get ForwardName() -> (ForwardName=None)
     "[1]":Name=Get SidewaysName() -> (SidewaysName=None)
     "[2]":Name=Get BackwardsName() -> (BackwardsName=None)
     "[3]":Name=Get ThrustName() -> (ThrustName=None)
    ) -> (Array:Array of Names)
   )
  ]}
]
variables [
 BackwardsName {type=string, default=None}
 ForwardName {type=string, default=None}
 SidewaysName {type=string, default=None}
 ThrustName {type=string, default=None}
]
widgets {
 info {rootWidgetClass=&ScaleBox, widgetCount=18}
 widgets [
  ScaleBox_1 {widgetClassPath=&ScaleBox, widget=ScaleBox_1}
  VerticalBox_71 {widgetClassPath=&VerticalBox, parent=ScaleBox_1, widget=VerticalBox_71, slot {class=ScaleBoxSlot}}
  EUW_DragControl_Literal {
   widgetClassPath=&/EUW_DragControl_Literal
   bIsVariable=true
   parent=VerticalBox_71
   widget=EUW_DragControl_Literal
   properties {
    metaSoundLiteralViewModel_Float_Backwards=&self:WidgetTree.EUW_DragControl_Literal.MetaSoundLiteralViewModel_Float_2
    metaSoundLiteralViewModel_Float_Forward=&self:WidgetTree.EUW_DragControl_Literal.MetaSoundLiteralViewModel_Float_1
    metaSoundLiteralViewModel_Float_Sideways=&self:WidgetTree.EUW_DragControl_Literal.MetaSoundLiteralViewModel_Float_0
    metaSoundLiteralViewModel_Float_Thrust=&self:WidgetTree.EUW_DragControl_Literal.MetaSoundLiteralViewModel_Float_3
   }
   slot {class=VerticalBoxSlot, properties {horizontalAlignment=HAlign_Left}}
  }
  HorizontalBox_206 {widgetClassPath=&HorizontalBox, parent=VerticalBox_71, widget=HorizontalBox_206, slot {class=VerticalBoxSlot}}
  VerticalBox_152 {
   widgetClassPath=&VerticalBox
   parent=HorizontalBox_206
   widget=VerticalBox_152
   slot {class=HorizontalBoxSlot, properties {horizontalAlignment=HAlign_Left, size {sizeRule=Fill, value=1}}}
  }
  HorizontalBox_208 {
   widgetClassPath=&HorizontalBox
   parent=VerticalBox_152
   widget=HorizontalBox_208
   slot {class=VerticalBoxSlot, properties {horizontalAlignment=HAlign_Left, size {sizeRule=Fill, value=1}}}
  }
  TextBlock_317 {
   widgetClassPath=&TextBlock
   parent=HorizontalBox_208
   widget=TextBlock_317
   properties {
    text="Forwards: "
    colorAndOpacity {colorUseRule=UseColor_Specified, specifiedColor=<rgba 0.5271 0.5271 0.5271 1>}
    font {
     bForceMonospaced=false
     bMaterialIsStencil=false
     fontMaterial=null
     fontObject=&Roboto
     letterSpacing=0
     monospacedWidth=1
     size=15
     skewAmount=0
     typefaceFontName=Default
     outlineSettings {
      bApplyOutlineToDropShadows=false
      bMiteredCorners=false
      bSeparateFillAlpha=false
      outlineMaterial=null
      outlineSize=0
      outlineColor=<rgba 0 0 0 1>
     }
    }
   }
   slot {class=HorizontalBoxSlot, properties {horizontalAlignment=HAlign_Left}}
  }
  ForwardsDisplay {
   widgetClassPath=&TextBlock
   bIsVariable=true
   parent=HorizontalBox_208
   widget=ForwardsDisplay
   properties {
    text=0
    font {
     bForceMonospaced=false
     bMaterialIsStencil=false
     fontMaterial=null
     fontObject=&Roboto
     letterSpacing=0
     monospacedWidth=1
     size=15
     skewAmount=0
     typefaceFontName=Regular
     outlineSettings {
      bApplyOutlineToDropShadows=false
      bMiteredCorners=false
      bSeparateFillAlpha=false
      outlineMaterial=null
      outlineSize=0
      outlineColor=<rgba 0 0 0 1>
     }
    }
   }
   slot {class=HorizontalBoxSlot, properties {horizontalAlignment=HAlign_Center}}
  }
  HorizontalBox {
   widgetClassPath=&HorizontalBox
   parent=VerticalBox_152
   widget=HorizontalBox
   slot {class=VerticalBoxSlot, properties {size {sizeRule=Fill, value=1}}}
  }
  TextBlock {
   widgetClassPath=&TextBlock
   parent=HorizontalBox
   widget=TextBlock
   properties {
    text="Backwards: "
    colorAndOpacity {colorUseRule=UseColor_Specified, specifiedColor=<rgba 0.5271 0.5271 0.5271 1>}
    font {
     bForceMonospaced=false
     bMaterialIsStencil=false
     fontMaterial=null
     fontObject=&Roboto
     letterSpacing=0
     monospacedWidth=1
     size=15
     skewAmount=0
     typefaceFontName=Regular
     outlineSettings {
      bApplyOutlineToDropShadows=false
      bMiteredCorners=false
      bSeparateFillAlpha=false
      outlineMaterial=null
      outlineSize=0
      outlineColor=<rgba 0 0 0 1>
     }
    }
   }
   slot {class=HorizontalBoxSlot, properties {horizontalAlignment=HAlign_Left}}
  }
  BackwardsDisplay {
   widgetClassPath=&TextBlock
   bIsVariable=true
   parent=HorizontalBox
   widget=BackwardsDisplay
   properties {
    text=0
    font {
     bForceMonospaced=false
     bMaterialIsStencil=false
     fontMaterial=null
     fontObject=&Roboto
     letterSpacing=0
     monospacedWidth=1
     size=15
     skewAmount=0
     typefaceFontName=Regular
     outlineSettings {
      bApplyOutlineToDropShadows=false
      bMiteredCorners=false
      bSeparateFillAlpha=false
      outlineMaterial=null
      outlineSize=0
      outlineColor=<rgba 0 0 0 1>
     }
    }
   }
   slot {class=HorizontalBoxSlot, properties {horizontalAlignment=HAlign_Center}}
  }
  VerticalBox {
   widgetClassPath=&VerticalBox
   parent=HorizontalBox_206
   widget=VerticalBox
   slot {class=HorizontalBoxSlot, properties {horizontalAlignment=HAlign_Left, size {sizeRule=Fill, value=1}}}
  }
  HorizontalBox_1 {
   widgetClassPath=&HorizontalBox
   parent=VerticalBox
   widget=HorizontalBox_1
   slot {class=VerticalBoxSlot, properties {horizontalAlignment=HAlign_Left, verticalAlignment=VAlign_Bottom}}
  }
  TextBlock_2 {
   widgetClassPath=&TextBlock
   parent=HorizontalBox_1
   widget=TextBlock_2
   properties {
    text="Sideways: "
    colorAndOpacity {colorUseRule=UseColor_Specified, specifiedColor=<rgba 0.5271 0.5271 0.5271 1>}
    font {
     bForceMonospaced=false
     bMaterialIsStencil=false
     fontMaterial=null
     fontObject=&Roboto
     letterSpacing=0
     monospacedWidth=1
     size=15
     skewAmount=0
     typefaceFontName=Regular
     outlineSettings {
      bApplyOutlineToDropShadows=false
      bMiteredCorners=false
      bSeparateFillAlpha=false
      outlineMaterial=null
      outlineSize=0
      outlineColor=<rgba 0 0 0 1>
     }
    }
   }
   slot {class=HorizontalBoxSlot, properties {horizontalAlignment=HAlign_Left}}
  }
  SidewaysDisplay {
   widgetClassPath=&TextBlock
   bIsVariable=true
   parent=HorizontalBox_1
   widget=SidewaysDisplay
   properties {
    text=0
    font {
     bForceMonospaced=false
     bMaterialIsStencil=false
     fontMaterial=null
     fontObject=&Roboto
     letterSpacing=0
     monospacedWidth=1
     size=15
     skewAmount=0
     typefaceFontName=Regular
     outlineSettings {
      bApplyOutlineToDropShadows=false
      bMiteredCorners=false
      bSeparateFillAlpha=false
      outlineMaterial=null
      outlineSize=0
      outlineColor=<rgba 0 0 0 1>
     }
    }
   }
   slot {class=HorizontalBoxSlot, properties {horizontalAlignment=HAlign_Center}}
  }
  HorizontalBox_2 {
   widgetClassPath=&HorizontalBox
   parent=VerticalBox
   widget=HorizontalBox_2
   slot {class=VerticalBoxSlot, properties {horizontalAlignment=HAlign_Left, size {sizeRule=Fill, value=1}}}
  }
  TextBlock_4 {
   widgetClassPath=&TextBlock
   parent=HorizontalBox_2
   widget=TextBlock_4
   properties {
    text="Thrust: "
    colorAndOpacity {colorUseRule=UseColor_Specified, specifiedColor=<rgba 0.5271 0.5271 0.5271 1>}
    font {
     bForceMonospaced=false
     bMaterialIsStencil=false
     fontMaterial=null
     fontObject=&Roboto
     letterSpacing=0
     monospacedWidth=1
     size=15
     skewAmount=0
     typefaceFontName=Regular
     outlineSettings {
      bApplyOutlineToDropShadows=false
      bMiteredCorners=false
      bSeparateFillAlpha=false
      outlineMaterial=null
      outlineSize=0
      outlineColor=<rgba 0 0 0 1>
     }
    }
   }
   slot {class=HorizontalBoxSlot, properties {horizontalAlignment=HAlign_Left}}
  }
  ThustDispaly {
   widgetClassPath=&TextBlock
   bIsVariable=true
   parent=HorizontalBox_2
   widget=ThustDispaly
   properties {
    text=0
    font {
     bForceMonospaced=false
     bMaterialIsStencil=false
     fontMaterial=null
     fontObject=&Roboto
     letterSpacing=0
     monospacedWidth=1
     size=15
     skewAmount=0
     typefaceFontName=Regular
     outlineSettings {
      bApplyOutlineToDropShadows=false
      bMiteredCorners=false
      bSeparateFillAlpha=false
      outlineMaterial=null
      outlineSize=0
      outlineColor=<rgba 0 0 0 1>
     }
    }
   }
   slot {class=HorizontalBoxSlot, properties {horizontalAlignment=HAlign_Center}}
  }
 ]
}
