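# Editor utility widget that gives presets of the MSS_Glider_Thrust_Loop MetaSound a custom editing panel for glider movement audio inputs.
# It stores the builder handed over when the preset widget is constructed.
# On Construct, it initializes Forward, Sideways, Backwards and Thrust editor viewmodels with that builder, then passes each to one shared drag control.
# The drag control maps these to the Glider direction inputs, with thrust named Glider.AnyDirection.
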EditorUtilityWidgetBlueprint /FortMetasound/Cosmetics/Widgets/EUW_MetasoundPreset_GliderThrust
graphs [
 EventGraph {nodes [
   Initialize Literal Widgets() {is=title K2Node_CustomEvent}
   > Set Literal Widget Input Viewmodels!(
    execute
    Literal Widget:Meta Sound Literal Widget Interface Interface=Get EUW_DragControl_Default() -> (EUW_DragControl_Default:EUW Drag Control Default Object Reference)
    MetaSound Viewmodel:Meta Sound Viewmodel Object Reference=Get MetaSoundEditorViewModel_Forward() -> (MetaSoundEditorViewModel_Forward:Meta Sound Editor Viewmodel Object Reference)
    __WorldContext:Object Reference
   ) -> (then) {is=title K2Node_CallFunction}
   > Set Literal Widget Input Viewmodels(
    Literal Widget=Get EUW_DragControl_Default()
    MetaSound Viewmodel=Get MetaSoundEditorViewModel_Sideways() -> (MetaSoundEditorViewModel_Sideways:Meta Sound Editor Viewmodel Object Reference)
   )
   > Set Literal Widget Input Viewmodels(
    Literal Widget=Get EUW_DragControl_Default()
    MetaSound Viewmodel=Get MetaSoundEditorViewModel_Backwards() -> (MetaSoundEditorViewModel_Backwards:Meta Sound Editor Viewmodel Object Reference)
   )
   > Set Literal Widget Input Viewmodels(
    Literal Widget=Get EUW_DragControl_Default()
    MetaSound Viewmodel=Get MetaSoundEditorViewModel_Thrust() -> (MetaSoundEditorViewModel_Thrust:Meta Sound Editor Viewmodel Object Reference)
   )
   Event Construct()
   > Initialize Builder!(
    execute
    self:Meta Sound Viewmodel Object Reference=[
     Get MetaSoundEditorViewModel_Forward()
     Get MetaSoundEditorViewModel_Sideways()
     Get MetaSoundEditorViewModel_Backwards()
     Get MetaSoundEditorViewModel_Thrust()
    ]
    InBuilder:Meta Sound Builder Base Object Reference=Get Builder() -> (Builder:Meta Sound Builder Base Object Reference)
   ) -> (then)
   > Initialize Literal Widgets() {is=title K2Node_CallFunction}
   n1 = Event On MetaSoundPreset Widget Constructed!() -> (OutputDelegate, then, Builder:Meta Sound Builder Base Object Reference) {is=title K2Node_Event}
   > Set Builder(Builder:Meta Sound Builder Base Object Reference=@n1.Builder) -> (Output_Get:Meta Sound Builder Base Object Reference)
  ]}
 GetEditorName {nodes [
   GetEditorName() {is=title K2Node_FunctionEntry}
   > Return Node!(execute, ReturnValue:Text)
  ]}
 GetIconBrushName {nodes [
   GetIconBrushName() {is=title K2Node_FunctionEntry}
   > Return Node(ReturnValue:Name=None)
  ]}
 GetSupportedMetaSounds {nodes [
   GetSupportedMetaSounds() {is=title K2Node_FunctionEntry}
   > Return Node!(
    execute
    bSupportAllPresets:bool=false
    ExcludedMetaSounds:Array of Meta Sound Document Interface Interfaces
    IncludedMetaSounds:Array of Meta Sound Document Interface Interfaces=Get Included Meta Sounds() -> (Included Meta Sounds:Array of Meta Sound Document Interface Interfaces)
   )
  ]}
 OnInitializingTab {nodes [
   On Initializing Tab() {is=title K2Node_FunctionEntry}
   > Return Node!(execute, TabInfo:Meta Sound Template Widget Tab Info Structure)
  ]}
]
variables [
 Builder {type=MetaSoundBuilderBase, default=null}
 Included Meta Sounds {type="array<object>", default [MetaSoundSource'/FortMetasound/Cosmetics/Glider/MSS_Glider_Thrust_Loop']}
]
widgets {
 info {rootWidgetClass=&StackBox, widgetCount=2}
 widgets [
  StackBox_0 {widgetClassPath=&StackBox, widget=StackBox_0, properties {bOverride_Cursor=true}}
  EUW_DragControl_Default {
   widgetClassPath=&/EUW_DragControl_Default
   bIsVariable=true
   parent=StackBox_0
   widget=EUW_DragControl_Default
   properties {
    backwardsName=Glider.Backwards
    forwardName=Glider.Forwards
    metaSoundLiteralViewModel_Float_Backwards=&self:WidgetTree.EUW_DragControl_Default.MetaSoundLiteralViewModel_Float_2
    metaSoundLiteralViewModel_Float_Forward=&self:WidgetTree.EUW_DragControl_Default.MetaSoundLiteralViewModel_Float_0
    metaSoundLiteralViewModel_Float_Sideways=&self:WidgetTree.EUW_DragControl_Default.MetaSoundLiteralViewModel_Float_1
    metaSoundLiteralViewModel_Float_Thrust=&self:WidgetTree.EUW_DragControl_Default.MetaSoundLiteralViewModel_Float_3
    sidewaysName=Glider.Sideways
    thrustName=Glider.AnyDirection
   }
   slot {class=StackBoxSlot}
  }
 ]
}
